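# Unlit, two-sided translucent smoke material for Niagara mesh particles on the MistRaven character.
# Panned, noise-distorted tendril and base smoke layers, shaped by flame, sphere-mask and caustic textures or math, take color from curve atlases.
# Opacity uses particle alpha with camera, view-angle and depth fades, the last skipped on ES3_1.
# Niagara dynamic parameters drive pan speed, UV offsets, noise strength and optional gradient or clamped interpolation World Position Offset.
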Material /BRCosmetics/Characters/Player/Female/Medium/Bodies/F_MED_MistRaven_Peck/FX/M_Character_MistRaven_Smoke
graph {
 properties {
  bAutomaticallySetUsageInEditor=false
  bUsedWithNiagaraMeshParticles=true
  blendMode=BLEND_Translucent
  shadingModel=MSM_Unlit
  twoSided=true
  usedShadingModels=Unlit
 }
 nodes [
  ComponentMask(None=OneMinus(None=NamedRerouteUsage())) {g=true}
  n1 = NamedRerouteUsage()
  n2 = NamedRerouteUsage()
  n3 = MaterialFunctionCall(X=ScalarParameter() {defaultValue=1, parameterName=Tendril_UMult}, Y=ScalarParameter() {defaultValue=1, parameterName=Tendril_VMult}) {materialFunction=&MakeFloat2}
  n4 = DynamicParameter() {parameterIndex=2, paramNames [TendrilTexture_NoiseStrength, TendrilTexture_Offset, BaseTexture_Offset, BaseTexture_NoiseStrength]}
  n5 = Constant()
  n6 = Panner(
   Coordinate=Add(
    A=Multiply(A=@n2, B=@n3.Result)
    B=MaterialFunctionCall(X=@n4.TendrilTexture_Offset, Y=@n5) {materialFunction=&MakeFloat2}.Result
   )
   Time=Multiply(A=NamedRerouteUsage()) {constB=2}
   Speed=NamedRerouteUsage()
  )
  n7 = TextureSampleParameter2D(UVs=MaterialFunctionCall(UVs=Add(A=@n1, B=@n6), Texture Scale=ScalarParameter() {defaultValue=1, parameterName=Tendrils UV Scale - From Center}) {materialFunction=&ScaleUVsByCenter}.UVs) {
   parameterName=TendrilTexture
   texture=&/T_FX_MistRaven_Character_Smoke
  }
  n8 = ComponentMask(None=Multiply(A=@n7.RGB, B=NamedRerouteUsage())) {r=true}
  n9 = CurveAtlasRowParameter(CurveTime=MaterialFunctionCall(In=@n8, Contrast=ScalarParameter() {defaultValue=0.2, parameterName=TendrilColorContrast}) {materialFunction=&CheapContrast}.Result) {
   atlas=/Game/Athena/LinearColorCurves/CAtlas_Contrail_MadameMoth_Color
   curve=/Game/Athena/LinearColorCurves/CV_Backpack_MistRaven_Smoke_Color_01
   defaultValue=17
   parameterName=TendrilColorCurve
  }
  Multiply(A=@n9, B=@n8)
  n10 = VectorParameter() {group=AddedWPO, parameterName=AddeWPODirection, sortPriority=1, defaultValue=<rgba 0 0 1 0>}
  n11 = TextureCoordinate()
  n12 = Power(
   Base=Multiply(
    A=ComponentMask(None=@n11) {r=true}
    B=ScalarParameter() {defaultValue=1, group=AddedWPO, parameterName=AddedWPO_Gradient_PreMult, sortPriority=3}
   ) {constB=5}
   Exp=ScalarParameter() {defaultValue=2, group=AddedWPO, parameterName=AddedWPO_GradientPower, sortPriority=4}
  ) {constExponent=5}
  n13 = ScalarParameter() {defaultValue=1, group=AddedWPO, parameterName=AddedWPO_Gradient_PostMult, sortPriority=3}
  n14 = Saturate(None=Subtract(A=Multiply(A=@n12, B=@n13), B=ScalarParameter() {group=AddedWPO, parameterName=AddedWPO_Subtract}))
  n15 = DynamicParameter() {desc=Parameters set by Niagara, parameterIndex=3, paramNames [AddedWPOMult, PanSpeedMult, "", ""]}
  n16 = ScalarParameter() {defaultValue=20, group=AddedWPO, parameterName=AddedWPO_Strength, sortPriority=2}
  n17 = Transform(None=Multiply(A=Multiply(A=@n10.RGB, B=@n14), B=Multiply(A=@n15.AddedWPOMult, B=@n16))) {transformSourceType=TRANSFORMSOURCE_World, transformType=TRANSFORM_Local}
  n18 = ScalarParameter() {defaultValue=1, parameterName=UseAddedWPO}
  n19 = DynamicParameter() {desc=Parameters set by Niagara, defaultValue=<rgba 0 1 1 1>, paramNames [X, Y, Z, InterpolatedMult]}
  n20 = MaterialFunctionCall(Float3=Subtract(
    A=MaterialFunctionCall(X=@n19.X, Y=@n19.X, Z=@n19.X) {materialFunction=&MakeFloat3}.Result
    B=ParticlePositionWS()
   )) {materialFunction=&BreakOutFloat3Components}
  n21 = AppendVector(A=@n20.R, B=@n20.R)
  n22 = ScalarParameter() {defaultValue=100, group=Interpolate, parameterName=Interpolate Clamp XY}
  n23 = Multiply(A=@n22) {constB=-1}
  n24 = ScalarParameter() {defaultValue=20, group=Interpolate, parameterName=Interpolate Clamp Z}
  n25 = Multiply(A=@n24) {constB=-1}
  n26 = ComponentMask(None=TextureCoordinate()) {r=true}
  n27 = ScalarParameter() {defaultValue=1, group=Interpolate, parameterName=Interpolate Power}
  n28 = LinearInterpolate(
   B=AppendVector(A=Clamp(None=@n21, Min=@n23, Max=@n22) {maxDefault=1000, minDefault=-1000}, B=Clamp(None=@n20.B, Min=@n25, Max=@n24) {minDefault=-1})
   Alpha=Multiply(A=@n19.InterpolatedMult, B=Power(Base=@n26, Exp=@n27))
  )
  n29 = ScalarParameter() {parameterName=Use interpolate WPO}
  NamedRerouteDeclaration(None=Add(A=LinearInterpolate(B=@n17, Alpha=@n18), B=LinearInterpolate(B=@n28, Alpha=@n29))) {name=WPO, nodeColor=<rgba 0.0078 1 0 1>}
  n30 = TextureCoordinate()
  NamedRerouteDeclaration(None=@n30) {name=BaseUVs, nodeColor=<rgba 1 0.7031 0 1>}
  n31 = NamedRerouteUsage()
  n32 = MaterialFunctionCall(X=ScalarParameter() {defaultValue=1, parameterName=BaseTexture_UMult}, Y=ScalarParameter() {defaultValue=1, parameterName=BaseTexture_VMult}) {materialFunction=&MakeFloat2}
  n33 = DynamicParameter() {parameterIndex=2, paramNames [TendrilTexture_NoiseStrength, TendrilTexture_Offset, BaseTexture_Offset, BaseTexture_NoiseStrength]}
  n34 = Constant()
  n35 = NamedRerouteUsage()
  n36 = NamedRerouteUsage()
  n37 = Add(
   A=Add(
    A=Multiply(A=@n31, B=@n32.Result)
    B=MaterialFunctionCall(X=@n33.BaseTexture_Offset, Y=@n34) {materialFunction=&MakeFloat2}.Result
   )
   B=Multiply(A=ComponentMask(None=@n35) {r=true}, B=Multiply(A=@n36) {constB=4}) {constB=2.5}
  )
  n38 = NamedRerouteUsage()
  n39 = NamedRerouteUsage()
  NamedRerouteDeclaration(None=TextureSampleParameter2D(UVs=Panner(Coordinate=@n37, Time=@n38, Speed=@n39)) {parameterName=MaskTexture, texture=/Game/Accessories/FORT_Backpacks/Backpack_M_MED_KeyChain/FX/T_Backpack_KeyChain_Flame}.R) {name=Fire, nodeColor=<rgba 1 0.8672 0 1>}
  n40 = NamedRerouteUsage()
  n41 = NamedRerouteUsage()
  n42 = NamedRerouteUsage()
  n43 = TextureSampleParameter2D(UVs=Add(A=Panner(Coordinate=@n40, Time=@n41, Speed=@n42), B=NamedRerouteUsage())) {parameterName=BaseTexture, samplerType=SAMPLERTYPE_LinearColor, texture=&T_NoisePack_LinearColor}
  NamedRerouteDeclaration(None=@n43.G) {name=BaseTexture_G, nodeColor=<rgba 0 1 0.7188 1>}
  n44 = TextureCoordinate() {vTiling=0.5}
  n45 = Sine(None=ComponentMask(None=@n44) {g=true})
  n46 = OneMinus(None=ComponentMask(None=@n44) {r=true})
  n47 = Multiply(
   A=NamedRerouteUsage()
   B=MaterialFunctionCall(
    X=ScalarParameter() {defaultValue=2.5, parameterName=BaseTexture_Noise2_UMult}
    Y=ScalarParameter() {defaultValue=1, parameterName=BaseTexture_Noise2_VMult}
   ) {materialFunction=&MakeFloat2}.Result
  )
  n48 = Multiply(A=NamedRerouteUsage()) {constB=1.25}
  n49 = NamedRerouteUsage()
  n50 = Multiply(
   A=TextureSample(UVs=Panner(Coordinate=@n47, Time=@n48, Speed=@n49)) {samplerType=SAMPLERTYPE_Grayscale, texture=&LowResBlurredNoise}.R
   B=ComponentMask(None=NamedRerouteUsage()) {r=true}
  )
  n51 = SphereMask(
   A=Add(A=TextureCoordinate() {uTiling=1.2, vTiling=1.2}, B=Multiply(A=@n50) {constB=0.5})
   B=MaterialFunctionCall(X=Constant() {r=1}, Y=Constant() {r=0.6}) {materialFunction=&MakeFloat2}.Result
  ) {attenuationRadius=0.95, hardnessPercent=0}
  n52 = Multiply(A=Power(Base=Multiply(A=@n45, B=@n46)) {constExponent=1.2}, B=Saturate(None=Multiply(A=@n51) {constB=5}))
  NamedRerouteDeclaration(None=@n52) {name=ShapeMask, nodeColor=<rgba 0 1 0.8594 1>}
  n53 = Multiply(
   A=NamedRerouteUsage()
   B=MaterialFunctionCall(
    X=ScalarParameter() {defaultValue=0.2, parameterName=TendirlNoise_UMult}
    Y=ScalarParameter() {defaultValue=2, parameterName=TendirlNoise_VMult}
   ) {materialFunction=&MakeFloat2}.Result
  )
  n54 = Multiply(A=NamedRerouteUsage())
  n55 = NamedRerouteUsage()
  n56 = Saturate(None=TextureSample(UVs=Panner(Coordinate=@n53, Time=@n54, Speed=@n55)) {samplerType=SAMPLERTYPE_Grayscale, texture=&LowResBlurredNoise}.R)
  n57 = ScalarParameter() {defaultValue=0.05, group=Noise, parameterName=TendirlNoiseStrength}
  n58 = DynamicParameter() {parameterIndex=2, paramNames [TendrilTexture_NoiseStrength, TendrilTexture_Offset, BaseTexture_Offset, BaseTexture_NoiseStrength]}
  NamedRerouteDeclaration(None=Multiply(A=LinearInterpolate(Alpha=@n56), B=Multiply(A=@n57, B=@n58.TendrilTexture_NoiseStrength))) {name=Tendril_Noise, nodeColor=<rgba 0 0.4609 1 1>}
  NamedRerouteDeclaration(None=@n30) {name=Masked UV Channel, nodeColor=<rgba 1 0.1172 0 1>}
  n59 = Time()
  n60 = ScalarParameter() {defaultValue=0.1, parameterName=OverallPanSpeed}
  NamedRerouteDeclaration(None=Multiply(
    A=Multiply(A=@n59, B=@n60)
    B=DynamicParameter() {desc=Parameters set by Niagara, parameterIndex=3, paramNames [AddedWPOMult, PanSpeedMult, "", ""]}.PanSpeedMult
   )) {name=PanTime, nodeColor=<rgba 1 0.5391 0 1>}
  NamedRerouteDeclaration(None=MaterialFunctionCall(X=ScalarParameter() {defaultValue=-1, parameterName=OverallPanSpeed_U}, Y=ScalarParameter() {parameterName=OverallPanSpeed_V}) {materialFunction=&MakeFloat2}.Result) {name=PanDirection, nodeColor=<rgba 0.2188 0 1 1>}
  n61 = CurveAtlasRowParameter(CurveTime=NamedRerouteUsage()) {
   atlas=&CAtlas_Contrail_MadameMoth_Color
   curve=/Game/Athena/LinearColorCurves/CV_Backpack_MistRaven_Flame_Color_02
   defaultValue=15
   parameterName=BaseSmokeColorCurve
  }
  NamedRerouteDeclaration(None=Add(A=@n61, B=@n9)) {name=Emissive, nodeColor=<rgba 1 0 0.8906 1>}
  n62 = Multiply(A=Multiply(A=@n43.R, B=Power(Base=@n52)), B=NamedRerouteUsage())
  n63 = Power(Base=LinearInterpolate(Alpha=@n62) {constB=2.5}, Exp=ScalarParameter() {defaultValue=1, parameterName=Power})
  n64 = MaterialFunctionCall(
   X=ScalarParameter() {defaultValue=1, parameterName=BaseTexture_Noise3_UMult}
   Y=ScalarParameter() {defaultValue=1, parameterName=BaseTexture_Noise3_VMult}
  ) {materialFunction=&MakeFloat2}
  n65 = Panner(
   Coordinate=Multiply(A=NamedRerouteUsage(), B=Multiply(A=@n64.Result) {constB=0.5}) {constB=0.5}
   Time=Multiply(A=NamedRerouteUsage()) {constB=0.5}
   Speed=NamedRerouteUsage()
  )
  n66 = Multiply(A=NamedRerouteUsage()) {constB=0.05}
  n67 = Multiply(A=TextureSampleParameter2D(UVs=Add(A=@n65, B=@n66)) {parameterName=Orange_Texture, texture=/Game/Global/Textures/Default/Caustics/T_Caustics_Softer}.R) {constB=2}
  NamedRerouteDeclaration(None=Add(A=@n62, B=Multiply(A=@n63, B=@n67))) {name=BaseSmoke, nodeColor=<rgba 0 0.8125 1 1>}
  n68 = ParticleColor()
  n69 = ScalarParameter() {defaultValue=25, parameterName=NearCamFadeDistance}
  n70 = Multiply(A=NamedRerouteUsage(), B=ScalarParameter() {defaultValue=0.25, parameterName=BaseSmoke_OpacityMult}) {constB=0.25}
  n71 = Multiply(
   A=MaterialFunctionCall("FadeSensitivity (optional)"=ScalarParameter() {defaultValue=1.4, parameterName=AngleFadeSensitivity}) {materialFunction=&FadeBasedOnViewAngle}.Result
   B=Multiply(
    A=MaterialFunctionCall(Distance=@n69) {materialFunction=&Near_Camera_Fade}.Result
    B=Add(A=@n8, B=@n70)
   )
  )
  n72 = FeatureLevelSwitch(Default=DepthFade(Opacity=@n71, FadeDistance=ScalarParameter() {defaultValue=5, parameterName=DepthFade}), ES3_1=@n71)
  NamedRerouteDeclaration(None=Multiply(A=Multiply(A=@n68.A, B=@n72)) {constB=10}) {name=Opacity, nodeColor=<rgba 1 0 0.2578 1>}
  n73 = NamedRerouteUsage()
  n74 = MaterialFunctionCall(
   X=ScalarParameter() {defaultValue=1.2, parameterName=BaseTexture_Noise_UMult}
   Y=ScalarParameter() {defaultValue=0.8, parameterName=BaseTexture_Noise_VMult}
  ) {materialFunction=&MakeFloat2}
  n75 = NamedRerouteUsage()
  n76 = TextureSample(UVs=Panner(Coordinate=Multiply(A=@n73, B=@n74.Result), Time=Multiply(A=@n75) {constB=1.5}, Speed=NamedRerouteUsage())) {texture=/BRCosmetics/Weapons/FORT_Melee/Pickaxe_CoatCheck/FX/T_Pickaxe_CoatCheck_Noise}
  n77 = ScalarParameter() {defaultValue=0.2, parameterName=BaseTexture_NoiseStrength}
  n78 = DynamicParameter() {parameterIndex=2, paramNames [TendrilTexture_NoiseStrength, TendrilTexture_Offset, BaseTexture_Offset, BaseTexture_NoiseStrength]}
  NamedRerouteDeclaration(None=Multiply(A=LinearInterpolate(Alpha=@n76.R), B=Multiply(A=@n77, B=@n78.BaseTexture_NoiseStrength))) {name=Base_Noise, nodeColor=<rgba 0 1 0.2969 1>}
  NamedRerouteUsage()
  NamedRerouteUsage()
  NamedRerouteUsage()
 ]
}
