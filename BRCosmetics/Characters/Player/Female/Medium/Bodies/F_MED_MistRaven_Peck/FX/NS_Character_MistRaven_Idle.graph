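# Niagara idle effect for the MistRaven player character, attaching dark, fire-spline-named mesh particles to skeletal mesh bones or sockets.
# Fire_spline002 bursts eight particles, using partition-indexed arrays for each one's bone, offset, rotation, scale and camera offset, with a separate Front End versus Gameplay size modifier.
# Fire_spline003 bursts three particles on one bone, mapping partition to VisibilityTag across mesh and sprite renderers.
# Both re-follow bones every update and feed randomized dynamic material parameters.
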NiagaraSystem /BRCosmetics/Characters/Player/Female/Medium/Bodies/F_MED_MistRaven_Peck/FX/NS_Character_MistRaven_Idle
dependencies {
 usedDataInterfaces [
  &NiagaraDataInterfaceArrayFloat3
  &NiagaraDataInterfaceArrayFloat
  &NiagaraDataInterfaceArrayInt32
  &NiagaraDataInterfaceSkeletalMesh
 ]
 usedDynamicInputs []
 usedModules [
  &self:SystemSpawnScript.SystemScriptSource.SystemScriptGraph.NiagaraNodeAssignment_0.SetVariables_#1
  &SystemState
 ]
}
emitters {
 Fire_spline002 {
  emitterUpdateScript [
   &EmitterState(
    Inactive Response=::NewEnumerator1
    Life Cycle Mode=::NewEnumerator1
    Loop Behavior=::NewEnumerator1
    Loop Duration Mode=::NewEnumerator1
    UseLoopDelay=false
   )
   &SpawnBurst_Instantaneous(Spawn Count=8, Spawn Time=0.1, Use Spawn Probability=false, UseLoopCountLimit=false)
  ]
  particleSpawnScript [
   &PartitionParticles(Number Of Partitions=8)
   &InitializeParticle(
    Color=<rgba 0.0208 0.0168 0.0169 0.5>
    Mesh Scale=&Multiply_VectorByFloat(
     Vector=&SelectVectorFromArray(
      Direct Array Index=$Particles.PartitionParticles.Partition
      Vector Selection Array=DataInterface() {propertyValues {FloatData [<1 1 1>, <1.5 1.5 1>, <1.2 1.2 1>, <1 1 0.65>, <2 2 0.8>, <1 1 0.8>, <1 1 0.5>, <2 2 0.8>]}}
     )
     Float=&Multiply_Float(
      A=0.2
      B=&SelectFloatFromArray(
       Direct Array Index=$Particles.PartitionParticles.Partition
       Float Selection Array=DataInterface() {propertyValues {FloatData [1,1,1,1,0.4,1,1,0.3]}}
      )
     )
    )
    Mesh Scale Mode=::NewEnumerator4
    UsePositionOffset=false
    Write Material Random=true
   )
   &InitialMeshOrientation(
    Mesh Orientation Mode=::NewEnumerator2
    Rotation=&Multiply_VectorByFloat(
     Vector=&SelectVectorFromArray(
      Direct Array Index=$Particles.PartitionParticles.Partition
      Vector Selection Array=DataInterface() {propertyValues {FloatData [<0.5 0 0>, <0.5 0 0>, <0.5 0 0>, <0 0 0>, <0 -0.1 0>, <0 -0.2 0>, <0 0.25 0>, <0 -0.18 0>]}}
     )
     Float=1
    )
    Use Rotation Vector=true
   )
   &InitialMeshOrientation(
    Mesh Orientation Mode=::NewEnumerator2
    Rotation=&Multiply_VectorByFloat(
     Vector=&SelectVectorFromArray(
      Direct Array Index=$Particles.PartitionParticles.Partition
      Vector Selection Array=DataInterface() {propertyValues {FloatData [<0 -0.15 0>, <0 -0.15 0>, <0 -0.15 0>, <0 -0.15 0>, <-0.05 0.1 0>, <0 0 0>, <0 0 0>, <-0.05 0.1 0>]}}
     )
     Float=1
    )
    Use Rotation Vector=true
   ) {module=InitialMeshOrientation001}
   &InitialMeshOrientation(
    Mesh Orientation Mode=::NewEnumerator2
    Rotation=&Multiply_VectorByFloat(
     Vector=&SelectVectorFromArray(
      Direct Array Index=$Particles.PartitionParticles.Partition
      Vector Selection Array=DataInterface() {propertyValues {FloatData [<0 0 -0.05>, <0 0 -0.05>, <0 0 -0.05>, <0 0 -0.05>, <0 0 0>, <0 0 0>, <0 0 0>, <0 0 0>]}}
     )
     Float=1
    )
    Use Rotation Vector=true
   ) {module=InitialMeshOrientation002}
   &CameraOffset(Camera Offset Amount=&SelectFloatFromArray(
     Direct Array Index=$Particles.PartitionParticles.Partition
     Float Selection Array=DataInterface() {propertyValues {FloatData [0,0,0,0,5,4,4,4]}}
    ))
  ]
  particleUpdateScript [
   &ParticleState(Kill Particles When Lifetime Has Elapsed=false, Let Infinitely Lived Particles Die When Emitter Deactivates=true)
   &SkeletalMeshLocation(
    Bone / Socket Index=&SelectIntFromArray(Direct Array Index=$Particles.PartitionParticles.Partition, Int Selection Array=DataInterface() {propertyValues {IntData [0,0,6,5,2,3,4,2]}})
    Bone Sampling Mode=::NewEnumerator4
    MaskSpawnGroup=false
    Sampled Position Offset=&SelectVectorFromArray(
     Direct Array Index=$Particles.PartitionParticles.Partition
     Vector Selection Array=DataInterface() {propertyValues {FloatData [<-20 0 -2>, <-10 0 -2>, <-38 0 -2>, <28 0 6>, <-2 -3 0>, <25 0 0.5>, <-10 0 0>, <4 0 -0.5>]}}
    )
    Skeletal Mesh=$System.SkeletalMesh
   )
   &UpdateMeshOrientation(Rotation Rate=20) {enabled=false}
   &ScaleMeshSize(Scale Factor=&Multiply_Vector(A=&FN_FrontEndVector3Modifier(Gameplay=<0.8 0.8 1>, Front End=<0.6 0.6 0.5>), B=<1.5 1.5 0.65>)) {module=ScaleMeshSize001}
   &ScaleColor(ScaleA=true, ScaleRGB=true)
   /SetVariables(Particles.Color1:LinearColor=<rgba 0.005 0.005 0.005 1.25>) {isSetParameters=true, module=SetVariables_#2}
   /SetVariables() {isSetParameters=true, module=SetVariables_#3}
   /SetVariables() {isSetParameters=true, module=SetVariables_#4}
   &DynamicMaterialParameters(
    Index 0 Param 1=&Lerp_Float(
     A=&SelectFloatFromArray(
      Direct Array Index=$Particles.PartitionParticles.Partition
      Float Selection Array=DataInterface() {propertyValues {FloatData [0.7,0.8,0.8,0.8]}}
     )
     B=0.8
     Alpha=0
    )
    Index 0 Param 2=$Engine.Time
    Index 0 Param 3=3
    Index 0 Param 4=&SelectFloatFromArray(
     Direct Array Index=$Particles.PartitionParticles.Partition
     Float Selection Array=DataInterface() {propertyValues {FloatData [0,0.5,0.25,0.5]}}
    )
    Index 2 Param 1=0.7
    Module.Index0Float0Write=true
    Module.Index0Float1Write=true
    Module.Index0Float2Write=true
    Module.Index0Float3Write=true
    Module.Index2Float0Write=true
    Module.Index2Float1Write=true
    Module.Index2Float2Write=true
    Module.Index2Float3Write=true
    Write Parameter Index 2=::NewEnumerator1
   ) {enabled=false}
   &DynamicMaterialParameters(
    Index 2 Param 1=1
    Index 2 Param 2=&RandomRangeFloat(Minimum=0, Maximum=1)
    Index 2 Param 3=&RandomRangeFloat(Minimum=0, Maximum=1)
    Index 2 Param 4=1
    Index 3 Param 1=&SelectFloatFromArray(
     Direct Array Index=$Particles.PartitionParticles.Partition
     Float Selection Array=DataInterface() {propertyValues {FloatData [1,1,1,1,0.5,0.2,0.5,0.5]}}
    )
    Index 3 Param 2=&RandomRangeFloat(Minimum=0.7, Maximum=1)
    Module.Param0WriteEnabled2=true
    Module.Param0WriteEnabled3=true
    Module.Param1WriteEnabled2=true
    Module.Param1WriteEnabled3=true
    Module.Param2WriteEnabled2=true
    Module.Param2WriteEnabled3=true
    Module.Param3WriteEnabled2=true
    Module.Param3WriteEnabled3=true
    Write Parameter Index 0=false
    Write Parameter Index 2=true
    Write Parameter Index 3=true
   ) {module=DynamicMaterialParameters001}
   /SetVariables(Particles.scale2:Vector3f=&Multiply_Vector(A=$Particles.Scale, B=&Multiply_VectorByFloat(Vector=<1 1 2>, Float=0.5))) {isSetParameters=true, module=SetVariables_#5}
   /SetVariables(
    Particles.dyn5:Vector4f=&MakeVector4(
     X=&RandomRangeFloat(Minimum=0, Maximum=1)
     Y=&RandomRangeFloat(Minimum=0, Maximum=1)
     Z=0
     W=0
    )
    Particles.dyn8:Vector4f=&MakeVector4(
     X=&RandomRangeFloat(Minimum=0, Maximum=1)
     Y=&RandomRangeFloat(Minimum=0, Maximum=1)
     Z=0
     W=0
    )
   ) {isSetParameters=true, module=SetVariables_#6}
  ]
  rendererClasses [&NiagaraMeshRendererProperties]
  renderers [{rendererClass=&NiagaraMeshRendererProperties, rendererIndex=0}]
 }
 Fire_spline003 {
  emitterUpdateScript [
   &EmitterState(
    Inactive Response=::NewEnumerator1
    Life Cycle Mode=::NewEnumerator1
    Loop Behavior=::NewEnumerator1
    Loop Duration Mode=::NewEnumerator1
    UseLoopDelay=false
   )
   &SpawnBurst_Instantaneous(Spawn Count=3, Spawn Time=0.1, Use Spawn Probability=false, UseLoopCountLimit=false)
  ]
  particleSpawnScript [
   &PartitionParticles(Number Of Partitions=3)
   &InitializeParticle(Mesh Scale Mode=::NewEnumerator4, UsePositionOffset=false, Write Material Random=true)
   &InitialMeshOrientation(Mesh Axis To Orient=<0 0 1>, Mesh Orientation Mode=::NewEnumerator2, Use Rotation Vector=true)
   &InitialMeshOrientation(
    Mesh Orientation Mode=::NewEnumerator2
    Rotation=&Multiply_VectorByFloat(
     Vector=&SelectVectorFromArray(
      Direct Array Index=$Particles.PartitionParticles.Partition
      Vector Selection Array=DataInterface() {propertyValues {FloatData [<0 -0.15 0>, <0 -0.15 0>, <0 -0.15 0>, <0 -0.15 0>, <-0.05 0.1 0>, <0 0 0>, <0 0 0>, <-0.05 0.1 0>]}}
     )
     Float=1
    )
    Use Rotation Vector=true
   ) {enabled=false, module=InitialMeshOrientation001}
   &InitialMeshOrientation(
    Mesh Orientation Mode=::NewEnumerator2
    Rotation=&Multiply_VectorByFloat(
     Vector=&SelectVectorFromArray(
      Direct Array Index=$Particles.PartitionParticles.Partition
      Vector Selection Array=DataInterface() {propertyValues {FloatData [<0 0 -0.05>, <0 0 -0.05>, <0 0 -0.05>, <0 0 -0.05>, <0 0 0>, <0 0 0>, <0 0 0>, <0 0 0>]}}
     )
     Float=1
    )
    Use Rotation Vector=true
   ) {enabled=false, module=InitialMeshOrientation002}
   &CameraOffset()
  ]
  particleUpdateScript [
   &ParticleState(Kill Particles When Lifetime Has Elapsed=false, Let Infinitely Lived Particles Die When Emitter Deactivates=true)
   &SkeletalMeshLocation(Bone / Socket Index=6, Bone Sampling Mode=::NewEnumerator4, MaskSpawnGroup=false, Skeletal Mesh=$System.SkeletalMesh)
   &UpdateMeshOrientation(Rotation Rate=20) {enabled=false}
   &ScaleMeshSize(Scale Factor=&Multiply_Vector(A=&FN_FrontEndVector3Modifier(Gameplay=<0.8 0.8 1>, Front End=<0.6 0.6 0.5>), B=<1.5 1.5 0.65>)) {enabled=false, module=ScaleMeshSize001}
   &ScaleColor(ScaleA=true, ScaleRGB=true)
   /SetVariables(Particles.Color1:LinearColor=<rgba 0.005 0.005 0.005 1.25>) {enabled=false, isSetParameters=true, module=SetVariables_#2}
   /SetVariables() {enabled=false, isSetParameters=true, module=SetVariables_#3}
   /SetVariables() {enabled=false, isSetParameters=true, module=SetVariables_#4}
   &DynamicMaterialParameters(
    Index 0 Param 1=&Lerp_Float(
     A=&SelectFloatFromArray(
      Direct Array Index=$Particles.PartitionParticles.Partition
      Float Selection Array=DataInterface() {propertyValues {FloatData [0.7,0.8,0.8,0.8]}}
     )
     B=0.8
     Alpha=0
    )
    Index 0 Param 2=$Engine.Time
    Index 0 Param 3=3
    Index 0 Param 4=&SelectFloatFromArray(
     Direct Array Index=$Particles.PartitionParticles.Partition
     Float Selection Array=DataInterface() {propertyValues {FloatData [0,0.5,0.25,0.5]}}
    )
    Index 2 Param 1=0.7
    Module.Index0Float0Write=true
    Module.Index0Float1Write=true
    Module.Index0Float2Write=true
    Module.Index0Float3Write=true
    Module.Index2Float0Write=true
    Module.Index2Float1Write=true
    Module.Index2Float2Write=true
    Module.Index2Float3Write=true
    Write Parameter Index 2=::NewEnumerator1
   ) {enabled=false}
   &DynamicMaterialParameters(
    Index 2 Param 1=1
    Index 2 Param 2=&RandomRangeFloat(Minimum=0, Maximum=1)
    Index 2 Param 3=&RandomRangeFloat(Minimum=0, Maximum=1)
    Index 2 Param 4=1
    Index 3 Param 2=&RandomRangeFloat(Minimum=0.7, Maximum=1)
    Module.Param0WriteEnabled2=true
    Module.Param0WriteEnabled3=true
    Module.Param1WriteEnabled2=true
    Module.Param1WriteEnabled3=true
    Module.Param2WriteEnabled2=true
    Module.Param2WriteEnabled3=true
    Module.Param3WriteEnabled2=true
    Module.Param3WriteEnabled3=true
    Write Parameter Index 0=false
    Write Parameter Index 2=true
    Write Parameter Index 3=true
   ) {module=DynamicMaterialParameters001}
   /SetVariables(Particles.scale2:Vector3f=&Multiply_Vector(A=$Particles.Scale, B=&Multiply_VectorByFloat(Vector=<1 1 2>, Float=0.5))) {enabled=false, isSetParameters=true, module=SetVariables_#5}
   /SetVariables(
    Particles.dyn5:Vector4f=&MakeVector4(
     X=&RandomRangeFloat(Minimum=0, Maximum=1)
     Y=&RandomRangeFloat(Minimum=0, Maximum=1)
     Z=0
     W=0
    )
    Particles.dyn8:Vector4f=&MakeVector4(
     X=&RandomRangeFloat(Minimum=0, Maximum=1)
     Y=&RandomRangeFloat(Minimum=0, Maximum=1)
     Z=0
     W=0
    )
   ) {enabled=false, isSetParameters=true, module=SetVariables_#6}
   /SetVariables(Particles.VisibilityTag=$Particles.PartitionParticles.Partition) {isSetParameters=true}
  ]
  rendererClasses [&NiagaraMeshRendererProperties, &NiagaraSpriteRendererProperties]
  renderers [
   {rendererClass=&NiagaraMeshRendererProperties, rendererIndex=0}
   {rendererClass=&NiagaraSpriteRendererProperties, rendererIndex=1}
   {rendererClass=&NiagaraMeshRendererProperties, rendererIndex=2}
   {rendererClass=&NiagaraMeshRendererProperties, rendererIndex=3}
  ]
 }
}
summary {
 systemName=NS_Character_MistRaven_Idle
 emitters [
  {emitterName=Fire_spline002, rendererClasses [&NiagaraMeshRendererProperties]}
  {emitterName=Fire_spline003, rendererClasses [&NiagaraMeshRendererProperties, &NiagaraSpriteRendererProperties]}
 ]
 userVariables [
  User.bisFootHidden {type=NiagaraBool}
  User.bIsFrontEnd {type=NiagaraBool}
  User.bIsFrontEndPreview {type=NiagaraBool}
  User.IsSkydiving {type=NiagaraBool}
 ]
}
