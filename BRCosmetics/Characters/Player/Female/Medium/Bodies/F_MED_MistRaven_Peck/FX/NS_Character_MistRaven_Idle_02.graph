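# Idle Niagara effect for the MistRaven Peck character body, attaching dark fire-style meshes with orange core parameters to the skeleton.
# Seventeen burst particles use partition-indexed arrays for bone, offset, orientation, scale and renderer tag; five more share one bone, each tagged separately.
# Smoothed positions feed a dynamic parameter, and color dims at night and in front end.
# User hand-hidden and foot-hidden flags zero the scale of matching particles in the first emitter.
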NiagaraSystem /BRCosmetics/Characters/Player/Female/Medium/Bodies/F_MED_MistRaven_Peck/FX/NS_Character_MistRaven_Idle_02
dependencies {
 usedDataInterfaces [
  &NiagaraDataInterfaceArrayFloat3
  &NiagaraDataInterfaceArrayFloat
  &NiagaraDataInterfaceArrayInt32
  &NiagaraDataInterfaceSkeletalMesh
 ]
 usedDynamicInputs []
 usedModules [
  &self:SystemSpawnScript.SystemScriptSource.SystemScriptGraph.NiagaraNodeAssignment_0.SetVariables_#1
  &SystemState
 ]
}
emitters {
 Fire_spline {
  emitterUpdateScript [
   &EmitterState(
    Inactive Response=::NewEnumerator1
    Life Cycle Mode=::NewEnumerator1
    Loop Behavior=::NewEnumerator1
    Loop Duration Mode=::NewEnumerator1
    UseLoopDelay=false
   )
   &SpawnBurst_Instantaneous(
    Spawn Count=&MakeIntFromFloat(Float=17)
    Spawn Time=0.1
    Use Spawn Probability=false
    UseLoopCountLimit=false
   )
  ]
  particleSpawnScript [
   &PartitionParticles(Number Of Partitions=17)
   &InitializeParticle(
    Color=<rgba 0.0208 0.0168 0.0169 1.6>
    Mesh Scale=&Multiply_VectorByFloat(
     Vector=&SelectVectorFromArray(
      Direct Array Index=$Particles.PartitionParticles.Partition
      Vector Selection Array=DataInterface() {propertyValues {FloatData [
         <1.5 1.5 1.25>
         <2 2 0.8>
         <1.2 1.2 1>
         <1.2 1.2 0.8>
         <1.8 1.8 0.7>
         <1 1 0.8>
         <0.8 0.8 0.8>
         <1.6 1.6 0.7>
         <1.25 1.25 1.1>
         <2.2 2.2 0.6>
         <1.2 1.2 1.2>
         <1 1 0.6>
         <1 1 0.4>
         <1 1 0.6>
         <1.5 1.5 1.4>
         <2 2 1>
         <1.25 1.25 0.6>
        ]}}
     )
     Float=&Multiply_Float(
      A=0.2
      B=&SelectFloatFromArray(
       Direct Array Index=$Particles.PartitionParticles.Partition
       Float Selection Array=DataInterface() {propertyValues {FloatData [1,1,1,1,0.4,1,1,0.6,1,1,1,1,1,1,1,1,1]}}
      )
     )
    )
    Mesh Scale Mode=::NewEnumerator4
    UsePositionOffset=false
    Write Material Random=true
   )
   &InitialMeshOrientation(
    Mesh Orientation Mode=::NewEnumerator2
    Rotation=&Multiply_VectorByFloat(
     Vector=&SelectVectorFromArray(
      Direct Array Index=$Particles.PartitionParticles.Partition
      Vector Selection Array=DataInterface() {propertyValues {FloatData [
         <0.5 0 0>
         <0.5 0 0>
         <0.5 0 0>
         <0 0 0>
         <0 -0.1 0>
         <0 -0.18 0>
         <0 0.25 0>
         <0 -0.18 0>
         <0.5 0 0>
         <0.5 0 0>
         <0 0 0>
         <0 0 0>
         <0 -0.1 0>
         <0 -0.2 0>
         <0.5 0 0>
         <0.5 0 0>
         <0 -0.18 0>
        ]}}
     )
     Float=1
    )
    Use Rotation Vector=true
   )
   &InitialMeshOrientation(
    Mesh Orientation Mode=::NewEnumerator2
    Rotation=&Multiply_VectorByFloat(
     Vector=&SelectVectorFromArray(
      Direct Array Index=$Particles.PartitionParticles.Partition
      Vector Selection Array=DataInterface() {propertyValues {FloatData [
         <0 -0.15 0>
         <0 -0.15 0>
         <0 -0.15 0>
         <0 -0.15 0>
         <0 0.05 0>
         <0 0 0>
         <0 0 0>
         <0 0.05 0>
         <0 -0.18 0>
         <0 -0.1 0>
         <0 -0.18 0>
         <0 -0.1 0>
         <0 0.05 0>
         <0 0.05 0>
         <0 -0.15 0>
         <0 0.025 0>
         <0 0.1 0>
        ]}}
     )
     Float=1
    )
    Use Rotation Vector=true
   ) {module=InitialMeshOrientation001}
   &InitialMeshOrientation(
    Mesh Orientation Mode=::NewEnumerator2
    Rotation=&Multiply_VectorByFloat(
     Vector=&SelectVectorFromArray(
      Direct Array Index=$Particles.PartitionParticles.Partition
      Vector Selection Array=DataInterface() {propertyValues {FloatData [
         <0 0 -0.05>
         <0 0 -0.05>
         <0 0 -0.05>
         <0 0 -0.05>
         <0 0 0>
         <0 0 0>
         <0 0 -0.025>
         <0 0 -0.05>
         <0 0 -0.05>
         <0 0 -0.05>
         <0 0 -0.05>
         <0 0 -0.05>
         <0 0 -0.05>
         <0 0 -0.05>
         <0 0 -0.05>
         <0 0 -0.05>
         <0 0 0.08>
        ]}}
     )
     Float=1
    )
    Use Rotation Vector=true
   ) {module=InitialMeshOrientation002}
   &CameraOffset(Camera Offset Amount=&SelectFloatFromArray(
     Direct Array Index=$Particles.PartitionParticles.Partition
     Float Selection Array=DataInterface() {propertyValues {FloatData [0,0,0,3,3,1,0,3,0,0,0,1,3,3,2,2,3]}}
    ))
   /SetVariables(Particles.VisibilityTag=&SelectIntFromArray(
     Direct Array Index=$Particles.PartitionParticles.Partition
     Int Selection Array=DataInterface() {propertyValues {IntData [0,0,0,0,2,1,1,2,0,0,0,0,2,2,0,0,2]}}
    )) {isSetParameters=true}
  ]
  particleUpdateScript [
   &ParticleState(Kill Particles When Lifetime Has Elapsed=false, Let Infinitely Lived Particles Die When Emitter Deactivates=true)
   &SkeletalMeshLocation(
    Bone / Socket Index=&SelectIntFromArray(
     Direct Array Index=$Particles.PartitionParticles.Partition
     Int Selection Array=DataInterface() {propertyValues {IntData [0,0,6,5,2,3,4,2,0,6,5,5,2,2,6,0,2]}}
    )
    Bone Sampling Mode=::NewEnumerator4
    MaskSpawnGroup=false
    Sampled Position Offset=&SelectVectorFromArray(
     Direct Array Index=$Particles.PartitionParticles.Partition
     Vector Selection Array=DataInterface() {propertyValues {FloatData [
        <-18 0 -2>
        <-12 0 -2>
        <-38 0 -2>
        <24 0 6>
        <-1 -1 0>
        <22 0 0.2>
        <-10 -1.5 0>
        <3 -1 1>
        <-23 0 -3>
        <-42 0 -2>
        <23 -1 7>
        <25 3 5>
        <-2 -1 -1>
        <5.5 -1 1>
        <-32 0 -3.5>
        <-3 0 2>
        <2 -1 -4>
       ]}}
    )
    Skeletal Mesh=$System.SkeletalMesh
   )
   &ScaleColor(
    Scale Alpha=&Multiply_Float(
     A=&SelectFloatFromArray(
      Direct Array Index=$Particles.PartitionParticles.Partition
      Float Selection Array=DataInterface() {propertyValues {FloatData [1,1,1,1,1,2,3,3,1,1,1,1,1,1,1,1,1]}}
     )
     B=1.15
    )
    Scale RGB=&FN_FrontEndVector3Modifier(
     Gameplay=&FN_Select_Vector_from_Time_Of_Day(Morning=<1 1 1>, Day=<1 1 1>, Evening=<1 1 1>, Night=<0.25 0.25 0.25>)
     Front End=<0.4 0.4 0.4>
    )
    ScaleA=true
    ScaleRGB=true
   ) {module=ScaleColor001}
   &ScaleMeshSize(Scale Factor=&Multiply_Vector(A=<0.8 0.8 1>, B=<1.2 1.2 0.65>)) {module=ScaleMeshSize001}
   &DynamicMaterialParameters(
    Index 0 Param 1=&RandomRangeFloat(Minimum=0, Maximum=1)
    Index 0 Param 2=&RandomRangeFloat(Minimum=0, Maximum=1)
    Index 0 Param 4=1
    Index 3 Param 1=&RandomRangeFloat(Minimum=0.5, Maximum=1)
    Index 3 Param 2=&RandomRangeFloat(Minimum=0, Maximum=1)
    Index 3 Param 3=&RandomRangeFloat(Minimum=0, Maximum=0)
    Module.Index0Param0Write=true
    Module.Index0Param1Write=true
    Module.Index0Param2Write=true
    Module.Index0Param3Write=false
    Module.Index3Param0Write=false
    Module.Index3Param1Write=true
    Module.Index3Param2Write=true
    Module.Index3Param3Write=false
    Write Parameter Index 3=true
   ) {module=DynamicMaterialParameters002}
   /SetVariables(Particles.CoreColor:LinearColor=<rgba 1 0.1123 0 2>) {isSetParameters=true}
   /SetVariables(Particles.CoreDynamic:Vector4f=&MakeVector4(X=1, Y=&RandomRangeFloat(Minimum=0, Maximum=1), Z=0, W=0)) {isSetParameters=true}
   &InterpolateOverTime(Rate Of Change=7, Target Vector Value:NiagaraPosition=$Particles.Position, Variable Type=::NewEnumerator2)
   /SetVariables(Particles.Interpolate Dynamic=&MakeFloat4FromVectorAndFloat("Vector (RGB)"=$Particles.InterpolateOverTime.InterpolateOverTime_V3_.Moving Average, "Float (Alpha)"=0.25)) {isSetParameters=true}
   &ScaleMeshSize(Scale Factor=&VectorFromFloat(Value=&Multiply_Float(
      A=&MakeCustomFloatFromBool(
       Bool=$User.bisHandHidden
       True Float=&SelectFloatFromArray(
        Direct Array Index=$Particles.PartitionParticles.Partition
        Float Selection Array=DataInterface() {propertyValues {FloatData [1,1,1,1,1,0,0,1,1,1,1,1,1,1,1,1,1]}}
       )
       False Float=1
      )
      B=&MakeCustomFloatFromBool(
       Bool=$User.bisFootHidden
       True Float=&SelectFloatFromArray(
        Direct Array Index=$Particles.PartitionParticles.Partition
        Float Selection Array=DataInterface() {propertyValues {FloatData [0,0,1,1,1,1,1,1,0,0,1,1,1,1,1,0,1]}}
       )
       False Float=1
      )
     )))
   /SetVariables(Particles.CoreScale:Vector3f=&Multiply_Vector(A=$Particles.Scale, B=<0.9 0.9 1>)) {isSetParameters=true}
  ]
  rendererClasses [&NiagaraMeshRendererProperties]
  renderers [
   {rendererClass=&NiagaraMeshRendererProperties, rendererIndex=0}
   {rendererClass=&NiagaraMeshRendererProperties, rendererIndex=1}
   {rendererClass=&NiagaraMeshRendererProperties, rendererIndex=2}
   {rendererClass=&NiagaraMeshRendererProperties, rendererIndex=3}
   {rendererClass=&NiagaraMeshRendererProperties, rendererIndex=4}
   {rendererClass=&NiagaraMeshRendererProperties, rendererIndex=5}
  ]
 }
 Fire_spline001 {
  emitterUpdateScript [
   &EmitterState(
    Inactive Response=::NewEnumerator1
    Life Cycle Mode=::NewEnumerator1
    Loop Behavior=::NewEnumerator1
    Loop Duration Mode=::NewEnumerator1
    UseLoopDelay=false
   )
   &SpawnBurst_Instantaneous(Spawn Count=5, Spawn Time=0.1, Use Spawn Probability=true, UseLoopCountLimit=false)
  ]
  particleSpawnScript [
   &PartitionParticles(Number Of Partitions=5)
   &InitializeParticle(
    Color=<rgba 0.0208 0.0168 0.0169 1.75>
    Mesh Scale=&VectorFromFloat(Value=1)
    Mesh Scale Mode=::NewEnumerator4
    UsePositionOffset=false
    Write Material Random=true
   )
   &InitialMeshOrientation(Mesh Axis To Orient=<0 0 1>, Mesh Orientation Mode=::NewEnumerator2, Use Rotation Vector=true)
   &CameraOffset(Camera Offset Amount=1)
  ]
  particleUpdateScript [
   &ParticleState(Kill Particles When Lifetime Has Elapsed=false, Let Infinitely Lived Particles Die When Emitter Deactivates=true)
   &SkeletalMeshLocation(Bone / Socket Index=6, Bone Sampling Mode=::NewEnumerator4, MaskSpawnGroup=false, Skeletal Mesh=$System.SkeletalMesh)
   &ScaleColor(
    Scale Alpha=1.5
    Scale RGB=&FN_FrontEndVector3Modifier(
     Gameplay=&FN_Select_Vector_from_Time_Of_Day(Morning=<1 1 1>, Day=<1 1 1>, Evening=<1 1 1>, Night=<0.25 0.25 0.25>)
     Front End=<0.4 0.4 0.4>
    )
    ScaleA=true
    ScaleRGB=true
   )
   /SetVariables(Particles.VisibilityTag=$Particles.PartitionParticles.Partition) {isSetParameters=true}
   &DynamicMaterialParameters(
    Index 0 Param 1=&RandomRangeFloat(Minimum=0, Maximum=1)
    Index 0 Param 2=&RandomRangeFloat(Minimum=0, Maximum=1)
    Index 0 Param 4=1
    Index 3 Param 1=&RandomRangeFloat(Minimum=0.5, Maximum=1)
    Index 3 Param 2=&RandomRangeFloat(Minimum=0, Maximum=1)
    Index 3 Param 3=&RandomRangeFloat(Minimum=0, Maximum=0)
    Module.Index0Param0Write=true
    Module.Index0Param1Write=true
    Module.Index0Param2Write=true
    Module.Index0Param3Write=false
    Module.Index3Param0Write=false
    Module.Index3Param1Write=true
    Module.Index3Param2Write=true
    Module.Index3Param3Write=false
    Write Parameter Index 3=true
   ) {module=DynamicMaterialParameters002}
   /SetVariables(Particles.cameraoffset2=5) {isSetParameters=true}
   &InterpolateOverTime(Rate Of Change=7, Target Vector Value:NiagaraPosition=$Particles.Position, Variable Type=::NewEnumerator2)
   /SetVariables(Particles.Interpolate Dynamic=&MakeFloat4FromVectorAndFloat("Vector (RGB)"=$Particles.InterpolateOverTime.InterpolateOverTime_V3_.Moving Average, "Float (Alpha)"=0.25)) {isSetParameters=true}
  ]
  rendererClasses [&NiagaraMeshRendererProperties]
  renderers [
   {rendererClass=&NiagaraMeshRendererProperties, rendererIndex=0}
   {rendererClass=&NiagaraMeshRendererProperties, rendererIndex=1}
   {rendererClass=&NiagaraMeshRendererProperties, rendererIndex=2}
   {rendererClass=&NiagaraMeshRendererProperties, rendererIndex=3}
   {rendererClass=&NiagaraMeshRendererProperties, rendererIndex=4}
  ]
 }
}
summary {
 systemName=NS_Character_MistRaven_Idle_02
 emitters [
  {emitterName=Fire_spline, rendererClasses [&NiagaraMeshRendererProperties]}
  {emitterName=Fire_spline001, rendererClasses [&NiagaraMeshRendererProperties]}
 ]
 userVariables [
  User.bisFootHidden {type=NiagaraBool}
  User.bIsFrontEnd {type=NiagaraBool}
  User.bIsFrontEndPreview {type=NiagaraBool}
  User.bisHandHidden {type=NiagaraBool}
 ]
}
