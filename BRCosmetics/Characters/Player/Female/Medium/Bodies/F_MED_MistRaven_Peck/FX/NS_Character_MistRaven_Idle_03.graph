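# Idle fire effect for the MistRaven character body, drawing dark mesh flames with orange core attributes attached to its skeletal mesh.
# Two emitters each burst six persistent particles in pairs across three bone or socket indices, with per-particle size, offset and orientation arrays.
# One follows the full mesh and the other the low mesh, with alpha switched by bisLowLOD.
# Smoothed position feeds material data; brightness varies by time of day and front end; bisHeadHidden sets visibility tags.
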NiagaraSystem /BRCosmetics/Characters/Player/Female/Medium/Bodies/F_MED_MistRaven_Peck/FX/NS_Character_MistRaven_Idle_03
dependencies {
 usedDataInterfaces [
  &NiagaraDataInterfaceArrayFloat3
  &NiagaraDataInterfaceArrayInt32
  &NiagaraDataInterfaceArrayFloat
  &NiagaraDataInterfaceCurve
  &NiagaraDataInterfaceSkeletalMesh
 ]
 usedDynamicInputs []
 usedModules [
  &SystemState
  &Timeline
  &self:SystemSpawnScript.SystemScriptSource.SystemScriptGraph.NiagaraNodeAssignment_7.SetVariables_#1
  &self:SystemSpawnScript.SystemScriptSource.SystemScriptGraph.NiagaraNodeAssignment_0.SetVariables_#2
 ]
}
emitters {
 Fire_spline {
  emitterUpdateScript [
   &EmitterState(
    Inactive Response=::NewEnumerator1
    Life Cycle Mode=::NewEnumerator1
    Loop Behavior=::NewEnumerator1
    Loop Duration Mode=::NewEnumerator1
    UseLoopDelay=false
   )
   &SpawnBurst_Instantaneous(Spawn Count=6, Spawn Time=0.1, Use Spawn Probability=false, UseLoopCountLimit=false)
  ]
  particleSpawnScript [
   &PartitionParticles(Number Of Partitions=6)
   /SetVariables(Particles.PartitionParticles.Partition=5) {enabled=false, isSetParameters=true}
   &InitializeParticle(
    Color=<rgba 0.0208 0.0168 0.0169 1.75>
    Mesh Scale=&Multiply_VectorByFloat(
     Vector=&SelectVectorFromArray(
      Direct Array Index=$Particles.PartitionParticles.Partition
      Vector Selection Array=DataInterface() {propertyValues {FloatData [<2 2 1>, <1.25 1.25 0.75>, <1.5 1.5 1>, <1.8 1.8 1.2>, <1.1 1.1 0.6>, <1.2 1.2 0.6>]}}
     )
     Float=0.08
    )
    Mesh Scale Mode=::NewEnumerator4
    UsePositionOffset=false
    Write Material Random=true
   )
   &InitialMeshOrientation(Mesh Orientation Mode=::NewEnumerator2, Rotation=<0.5 0 0>, Use Rotation Vector=true) {enabled=false}
   &InitialMeshOrientation(Mesh Orientation Mode=::NewEnumerator2, Rotation=<0 -0.25 0>, Use Rotation Vector=true) {enabled=false, module=InitialMeshOrientation001}
   &InitialMeshOrientation(
    Mesh Orientation Mode=::NewEnumerator2
    Rotation=&SelectVectorFromArray(
     Direct Array Index=$Particles.PartitionParticles.Partition
     Vector Selection Array=DataInterface() {propertyValues {FloatData [<0 0 0>, <0 0 0.035>, <0 -0.1 0>, <0 -0.1 0.04>, <0 -0.05 -0.02>, <0 -0.05 0>]}}
    )
    Use Rotation Vector=true
   ) {enabled=false, module=InitialMeshOrientation002}
   &CameraOffset(Camera Offset Amount=1)
   &InitialMeshOrientation(
    Mesh Orientation Mode=::NewEnumerator2
    Rotation=&SelectVectorFromArray(
     Direct Array Index=$Particles.PartitionParticles.Partition
     Vector Selection Array=DataInterface() {propertyValues {FloatData [<0.25 0 0>, <0.25 0 0>, <0.23 0 0>, <0.27 0 0>, <0.12 0 0>, <0.18 0 0>]}}
    )
    Use Rotation Vector=true
   ) {module=InitialMeshOrientation003}
   &InitialMeshOrientation(
    Mesh Orientation Mode=::NewEnumerator2
    Rotation=&SelectVectorFromArray(
     Direct Array Index=$Particles.PartitionParticles.Partition
     Vector Selection Array=DataInterface() {propertyValues {FloatData [<0 0 0>, <0 0 0>, <0 0 0.05>, <0 0 0.05>, <0 0 0.05>, <0 0 0.05>]}}
    )
    Use Rotation Vector=true
   ) {module=InitialMeshOrientation004}
  ]
  particleUpdateScript [
   &ParticleState(Kill Particles When Lifetime Has Elapsed=false, Let Infinitely Lived Particles Die When Emitter Deactivates=true)
   &SkeletalMeshLocation(
    Bone / Socket Index=&SelectIntFromArray(Direct Array Index=$Particles.PartitionParticles.Partition, Int Selection Array=DataInterface() {propertyValues {IntData [0,0,1,1,2,2]}})
    MaskSpawnGroup=false
    Mesh Sampling Type=::NewEnumerator5
    Sampled Position Offset=&SelectVectorFromArray(
     Direct Array Index=$Particles.PartitionParticles.Partition
     Vector Selection Array=DataInterface() {propertyValues {FloatData [<-2 0 0>, <-2 -2 0>, <-5 1 -4>, <-5 1 1.5>, <-3 0 0>, <-3 -5 -2>]}}
    )
    Skeletal Mesh=$System.SkeletalMeshLow
    Socket Sampling Mode=::NewEnumerator5
   )
   &ScaleColor(
    Scale Alpha=&Multiply_Float(
     A=&SelectFloatFromArray(Direct Array Index=$Particles.PartitionParticles.Partition, Float Selection Array=DataInterface() {propertyValues {FloatData [2,2,5,5,2,2]}})
     B=&MakeCustomFloatFromBool(Bool=$User.bisLowLOD, True Float=1, False Float=0)
    )
    Scale RGB=&FN_FrontEndVector3Modifier(
     Gameplay=&FN_Select_Vector_from_Time_Of_Day(Morning=<1 1 1>, Day=<1 1 1>, Evening=<1 1 1>, Night=<0.25 0.25 0.25>)
     Front End=<0.4 0.4 0.4>
    )
    ScaleA=true
    ScaleRGB=true
   )
   &DynamicMaterialParameters(
    Index 0 Param 1=&RandomRangeFloat(Minimum=0, Maximum=1)
    Index 0 Param 2=&RandomRangeFloat(Minimum=0, Maximum=1)
    Index 0 Param 4=1
    Index 3 Param 1=&RandomRangeFloat(Minimum=0.5, Maximum=1)
    Index 3 Param 2=&RandomRangeFloat(Minimum=0, Maximum=1)
    Index 3 Param 3=&RandomRangeFloat(Minimum=0, Maximum=0)
    Module.Index0Param0Write=true
    Module.Index0Param1Write=true
    Module.Index0Param2Write=true
    Module.Index0Param3Write=false
    Module.Index3Param0Write=false
    Module.Index3Param1Write=true
    Module.Index3Param2Write=true
    Module.Index3Param3Write=false
    Write Parameter Index 3=true
   ) {module=DynamicMaterialParameters002}
   &ScaleMeshSize(Scale Factor=&Multiply_VectorByFloat(Vector=$Particles.Initial.Scale, Float=&Lerp_Float(A=1, B=1, Alpha=$System.Timeline.NormalizedTime))) {enabled=false}
   /SetVariables(Particles.CoreColor:LinearColor=&MakeLinearColorFromVectorAndFloat(
     "Vector (RGB)"=&MakeVectorFromLinearColorRGB(Color=<rgba 1 0.1123 0 2>)
     "Float (Alpha)"=&Multiply_Float(
      A=&MakeFloatFromLinearColor(LinearColor=$Particles.Color, Channel:ENiagaraFloatFromLinearColorOptions=::NewEnumerator3)
      B=&SelectFloatFromArray(
       Direct Array Index=$Particles.PartitionParticles.Partition
       Float Selection Array=DataInterface() {propertyValues {FloatData [0.5,0.5,0.25,0.25,0.5,0.5]}}
      )
     )
    )) {isSetParameters=true, module=SetVariables_#3}
   /SetVariables(Particles.CoreDynamic:Vector4f=&MakeVector4(X=1, Y=&RandomRangeFloat(Minimum=0, Maximum=1), Z=0, W=0)) {isSetParameters=true, module=SetVariables_#4}
   /SetVariables(Particles.Scale=&Multiply_VectorByFloat(Vector=$Particles.Initial.Scale, Float=&Lerp_Float(A=1, B=1, Alpha=$System.Timeline.Float))) {enabled=false, isSetParameters=true}
   /SetVariables(Particles.CoreScale:Vector3f=&Multiply_Vector(
     A=$Particles.Scale
     B=&SelectVectorFromArray(
      Direct Array Index=$Particles.PartitionParticles.Partition
      Vector Selection Array=DataInterface() {propertyValues {FloatData [<1 1 1.35>, <1 1 1.2>, <1.5 1.5 1.5>, <1.5 1.15 1.2>, <1 1 0.8>, <1 1 1>]}}
     )
    )) {isSetParameters=true, module=SetVariables_#5}
   &InterpolateOverTime(Rate Of Change=7, Target Vector Value:NiagaraPosition=$Particles.Position, Variable Type=::NewEnumerator2)
   /SetVariables(Particles.Interpolate Dynamic=&MakeFloat4FromVectorAndFloat("Vector (RGB)"=$Particles.InterpolateOverTime.InterpolateOverTime_V3_.Moving Average, "Float (Alpha)"=0.25)) {isSetParameters=true, module=SetVariables_#6}
   /SetVariables(Particles.VisibilityTag=&MakeCustomIntFromBool(Bool=$User.bisHeadHidden, True INT=1, False INT=0)) {isSetParameters=true, module=SetVariables_#7}
  ]
  rendererClasses [&NiagaraMeshRendererProperties]
  renderers [
   {rendererClass=&NiagaraMeshRendererProperties, rendererIndex=0}
   {rendererClass=&NiagaraMeshRendererProperties, rendererIndex=1}
  ]
 }
 Fire_spline002 {
  emitterUpdateScript [
   &EmitterState(
    Inactive Response=::NewEnumerator1
    Life Cycle Mode=::NewEnumerator1
    Loop Behavior=::NewEnumerator1
    Loop Duration Mode=::NewEnumerator1
    UseLoopDelay=false
   )
   &SpawnBurst_Instantaneous(Spawn Count=6, Spawn Time=0.1, Use Spawn Probability=false, UseLoopCountLimit=false)
  ]
  particleSpawnScript [
   &PartitionParticles(Number Of Partitions=6)
   &InitializeParticle(
    Color=<rgba 0.0208 0.0168 0.0169 1.75>
    Mesh Scale=&Multiply_VectorByFloat(
     Vector=&SelectVectorFromArray(
      Direct Array Index=$Particles.PartitionParticles.Partition
      Vector Selection Array=DataInterface() {propertyValues {FloatData [<2 2 1>, <1.25 1.25 0.75>, <1.5 1.5 1>, <1.8 1.8 1.2>, <1.1 1.1 0.6>, <1.2 1.2 0.6>]}}
     )
     Float=0.08
    )
    Mesh Scale Mode=::NewEnumerator4
    UsePositionOffset=false
    Write Material Random=true
   )
   &InitialMeshOrientation(Mesh Orientation Mode=::NewEnumerator2, Rotation=<0.5 0 0>, Use Rotation Vector=true)
   &InitialMeshOrientation(Mesh Orientation Mode=::NewEnumerator2, Rotation=<0 -0.25 0>, Use Rotation Vector=true) {module=InitialMeshOrientation001}
   &InitialMeshOrientation(
    Mesh Orientation Mode=::NewEnumerator2
    Rotation=&SelectVectorFromArray(
     Direct Array Index=$Particles.PartitionParticles.Partition
     Vector Selection Array=DataInterface() {propertyValues {FloatData [<0 0 0>, <0 0 0.035>, <0 -0.1 0>, <0 -0.1 0.04>, <0 -0.05 -0.02>, <0 -0.05 0>]}}
    )
    Use Rotation Vector=true
   ) {module=InitialMeshOrientation002}
   &CameraOffset(Camera Offset Amount=1)
   &FN_SetSpawnGroup(Particles Per Group=6, Pattern=::NewEnumerator0) {enabled=false}
   /SetVariables(Particles.SpawnGroup=1) {isSetParameters=true}
  ]
  particleUpdateScript [
   &ParticleState(Kill Particles When Lifetime Has Elapsed=false, Let Infinitely Lived Particles Die When Emitter Deactivates=true)
   &SkeletalMeshLocation(
    Bone / Socket Index=&SelectIntFromArray(Direct Array Index=$Particles.PartitionParticles.Partition, Int Selection Array=DataInterface() {propertyValues {IntData [0,0,1,1,2,2]}})
    Bone Sampling Mode=::NewEnumerator4
    MaskSpawnGroup=true
    Sampled Position Offset=&SelectVectorFromArray(
     Direct Array Index=$Particles.PartitionParticles.Partition
     Vector Selection Array=DataInterface() {propertyValues {FloatData [<0 0 0>, <0 -2 0>, <0 3 -4>, <0 -2 -4>, <0 0 -4>, <3 3 -2>]}}
    )
    Skeletal Mesh=$System.SkeletalMesh
   )
   &ScaleColor(
    Scale Alpha=&Multiply_Float(
     A=&SelectFloatFromArray(Direct Array Index=$Particles.PartitionParticles.Partition, Float Selection Array=DataInterface() {propertyValues {FloatData [2,2,5,5,2,2]}})
     B=&MakeCustomFloatFromBool(Bool=$User.bisLowLOD, True Float=0, False Float=1)
    )
    Scale RGB=&FN_FrontEndVector3Modifier(
     Gameplay=&FN_Select_Vector_from_Time_Of_Day(Morning=<1 1 1>, Day=<1 1 1>, Evening=<1 1 1>, Night=<0.25 0.25 0.25>)
     Front End=<0.4 0.4 0.4>
    )
    ScaleA=true
    ScaleRGB=true
   )
   &DynamicMaterialParameters(
    Index 0 Param 1=&RandomRangeFloat(Minimum=0, Maximum=1)
    Index 0 Param 2=&RandomRangeFloat(Minimum=0, Maximum=1)
    Index 0 Param 4=1
    Index 3 Param 1=&RandomRangeFloat(Minimum=0.5, Maximum=1)
    Index 3 Param 2=&RandomRangeFloat(Minimum=0, Maximum=1)
    Index 3 Param 3=&RandomRangeFloat(Minimum=0, Maximum=0)
    Module.Index0Param0Write=true
    Module.Index0Param1Write=true
    Module.Index0Param2Write=true
    Module.Index0Param3Write=false
    Module.Index3Param0Write=false
    Module.Index3Param1Write=true
    Module.Index3Param2Write=true
    Module.Index3Param3Write=false
    Write Parameter Index 3=true
   ) {module=DynamicMaterialParameters002}
   /SetVariables(Particles.CoreColor:LinearColor=&MakeLinearColorFromVectorAndFloat(
     "Vector (RGB)"=&MakeVectorFromLinearColorRGB(Color=<rgba 1 0.1123 0 2>)
     "Float (Alpha)"=&Multiply_Float(
      A=&MakeFloatFromLinearColor(LinearColor=$Particles.Color, Channel:ENiagaraFloatFromLinearColorOptions=::NewEnumerator3)
      B=&SelectFloatFromArray(
       Direct Array Index=$Particles.PartitionParticles.Partition
       Float Selection Array=DataInterface() {propertyValues {FloatData [0.5,0.5,0.25,0.25,0.5,0.5]}}
      )
     )
    )) {isSetParameters=true, module=SetVariables_#3}
   /SetVariables(Particles.CoreDynamic:Vector4f=&MakeVector4(X=1, Y=&RandomRangeFloat(Minimum=0, Maximum=1), Z=0, W=0)) {isSetParameters=true, module=SetVariables_#4}
   /SetVariables(Particles.CoreScale:Vector3f=&Multiply_Vector(
     A=$Particles.Scale
     B=&SelectVectorFromArray(
      Direct Array Index=$Particles.PartitionParticles.Partition
      Vector Selection Array=DataInterface() {propertyValues {FloatData [<1 1 1.35>, <1 1 1.2>, <1.5 1.5 1.5>, <1.5 1.15 1.2>, <1 1 0.8>, <1 1 1>]}}
     )
    )) {isSetParameters=true, module=SetVariables_#5}
   &InterpolateOverTime(Rate Of Change=7, Target Vector Value:NiagaraPosition=$Particles.Position, Variable Type=::NewEnumerator2)
   /SetVariables(Particles.Interpolate Dynamic=&MakeFloat4FromVectorAndFloat("Vector (RGB)"=$Particles.InterpolateOverTime.InterpolateOverTime_V3_.Moving Average, "Float (Alpha)"=0.25)) {isSetParameters=true, module=SetVariables_#6}
   &ScaleColor(
    Scale Alpha=&MakeCustomFloatFromBool(Bool=$User.bisLowLOD, True Float=1, False Float=0)
    ScaleA=true
    ScaleRGB=true
   ) {enabled=false, module=ScaleColor001}
   /SetVariables(Particles.Position=&Lerp_Vector(
     A=Linked() {linkedVariable {name=Particles.SkeletalMeshLocation.SampledPosition, type=NiagaraPosition}}
     B=Linked() {linkedVariable {name=Particles.SkeletalMeshLocation001.SampledPosition, type=NiagaraPosition}}
     Alpha=&MakeCustomFloatFromBool(Bool=$User.bisLowLOD, True Float=1, False Float=0)
    )) {enabled=false, isSetParameters=true}
   /SetVariables(Particles.MeshOrientation=&MultiplyQuaternion(Quaternion A=$Particles.InitialRotation, Quaternion B=$Particles.SkeletalMeshLocation.SampledRotation)) {enabled=false, isSetParameters=true}
   /SetVariables(Particles.Scale=&Multiply_VectorByFloat(Vector=$Particles.Initial.Scale, Float=&Lerp_Float(A=0, B=1, Alpha=$System.Timeline.InverseNormalizedTime))) {enabled=false, isSetParameters=true}
   /SetVariables(Particles.VisibilityTag=&MakeCustomIntFromBool(Bool=$User.bisHeadHidden, True INT=1, False INT=0)) {isSetParameters=true, module=SetVariables_#7}
  ]
  rendererClasses [&NiagaraMeshRendererProperties]
  renderers [
   {rendererClass=&NiagaraMeshRendererProperties, rendererIndex=0}
   {rendererClass=&NiagaraMeshRendererProperties, rendererIndex=1}
  ]
 }
}
summary {
 systemName=NS_Character_MistRaven_Idle_03
 emitters [
  {emitterName=Fire_spline002, rendererClasses [&NiagaraMeshRendererProperties]}
  {emitterName=Fire_spline, rendererClasses [&NiagaraMeshRendererProperties]}
 ]
 userVariables [
  User.bIsFrontEnd {type=NiagaraBool}
  User.bIsFrontEndPreview {type=NiagaraBool}
  User.bisHeadHidden {type=NiagaraBool}
  User.bisLowLOD {type=NiagaraBool}
 ]
}
