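# Physics asset for the female medium MistRaven Peck cosmetic body, built around four hanging belt chains: outer and mid, left and right.
# Each chain runs three capsules from a small master body jointed to spine_02, with tight twist and mostly single-plane swing limits.
# Free-rotation cross-links tie neighboring chains together at their second and third segments.
# Pelvis, lower spine and right thigh capsules supply body volume; the left thigh has none.
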PhysicsAsset /BRCosmetics/Characters/Player/Female/Medium/Bodies/F_MED_MistRaven_Peck/Meshes/F_MED_MistRaven_Peck_Physics
bodies {
 belt_01_L [{shapeType=Capsule, length=9, radius=2, shapeName=belt_01_L_capsule, center=<6.1293 0 0>, rotation=<rot -90 -179.9999 -179.9999>}]
 belt_01_R [{shapeType=Capsule, length=5, radius=2, shapeName=belt_01_R_capsule, center=<6.1293 0 0>, rotation=<rot -90 -10.025 10.0249>}]
 belt_02_L [{shapeType=Capsule, length=5, radius=2, shapeName=belt_02_L_capsule, center=<6.1293 0 0>, rotation=<rot -90 -179.9901 179.9902>}]
 belt_02_R [{shapeType=Capsule, length=9.5, radius=2, shapeName=belt_02_R_capsule, center=<6.8903 0 0>, rotation=<rot -90 -90 89.9997>}]
 belt_03_L [{shapeType=Capsule, length=10, radius=2, shapeName=belt_03_L_capsule, center=<5.9457 -2.1641 0>, rotation=<rot -90 0 -19.9999>}]
 belt_03_R [{shapeType=Capsule, length=5, radius=2, shapeName=belt_03_R_capsule, center=<6.1293 0 0>, rotation=<rot -90 179.6834 -179.6837>}]
 belt_MSTR_L [{shapeType=Capsule, length=0, radius=0.5, shapeName=belt_MSTR_L_capsule, rotation=<rot -90 -44.9396 44.9398>}]
 belt_MSTR_R [{shapeType=Capsule, length=0, radius=0.5, shapeName=belt_MSTR_R_capsule, rotation=<rot -90 -45.7099 45.7098>}]
 belt_mid_01_L [{shapeType=Capsule, length=8.5, radius=2, shapeName=belt_mid_01_L_capsule, center=<6.1293 0 0>, rotation=<rot -90 -19.4711 19.4713>}]
 belt_mid_01_R [{shapeType=Capsule, length=5, radius=2, shapeName=belt_mid_01_R_capsule, center=<6.1293 0 0>, rotation=<rot -90 -172.4476 172.4477>}]
 belt_mid_02_L [{shapeType=Capsule, length=6.5, radius=2, shapeName=belt_mid_02_L_capsule, center=<6.1293 0 0>, rotation=<rot -90 -179.9605 179.9604>}]
 belt_mid_02_R [{shapeType=Capsule, length=8, radius=2, shapeName=belt_mid_02_R_capsule, center=<6.1293 0 0>, rotation=<rot -90 -179.9998 179.9996>}]
 belt_mid_03_L [{shapeType=Capsule, length=10.5, radius=2, shapeName=belt_mid_03_L_capsule, center=<5.9824 -1.0549 0>, rotation=<rot -90 0 -10.0003>}]
 belt_mid_03_R [{shapeType=Capsule, length=11.5, radius=2, shapeName=belt_mid_03_R_capsule, center=<6.1979 -1.0929 0>, rotation=<rot -90 -176.4237 166.4233>}]
 belt_mid_MSTR_L [{shapeType=Capsule, length=0, radius=0.5, shapeName=belt_mid_MSTR_L_capsule, rotation=<rot -90 -43.8709 43.871>}]
 belt_mid_MSTR_R [{shapeType=Capsule, length=0, radius=0.5, shapeName=belt_mid_MSTR_R_capsule, rotation=<rot -90 -46.4683 46.4683>}]
 pelvis [{shapeType=Capsule, length=12.4229, radius=8.5316, shapeName=pelvis_capsule, center=<0 -2.0692 0>}]
 spine_01 [{shapeType=Capsule, length=8.2089, radius=6.7303, shapeName=spine_01_capsule}]
 spine_02 [{shapeType=Capsule, length=8.6252, radius=6.1796, shapeName=spine_02_capsule}]
 thigh_r [{shapeType=Capsule, length=23.6947, radius=7.4946, shapeName=thigh_r_capsule, center=<20 -0.0001 0>, rotation=<rot -90 -160.5288 160.5286>}]
}
constraints [
 {
  bone1Name=spine_01
  bone2Name=pelvis
  swing1LimitDegrees=45
  swing1Motion=Free
  swing2LimitDegrees=45
  swing2Motion=Free
  twistLimitDegrees=45
  twistMotion=Free
 }
 {bone1Name=belt_01_L, bone2Name=belt_MSTR_L, swing1LimitDegrees=0, swing2LimitDegrees=45, twistLimitDegrees=5}
 {bone1Name=belt_02_L, bone2Name=belt_01_L, swing1LimitDegrees=0, swing2LimitDegrees=45, twistLimitDegrees=5}
 {bone1Name=belt_03_L, bone2Name=belt_02_L, swing1LimitDegrees=5, swing2LimitDegrees=45, twistLimitDegrees=5}
 {bone1Name=belt_01_R, bone2Name=belt_MSTR_R, swing1LimitDegrees=0, swing2LimitDegrees=45, twistLimitDegrees=5}
 {bone1Name=belt_02_R, bone2Name=belt_01_R, swing1LimitDegrees=0, swing2LimitDegrees=45, twistLimitDegrees=5}
 {bone1Name=belt_03_R, bone2Name=belt_02_R, swing1LimitDegrees=25, swing2LimitDegrees=45, twistLimitDegrees=5}
 {bone1Name=belt_mid_01_L, bone2Name=belt_mid_MSTR_L, swing1LimitDegrees=0, swing2LimitDegrees=45, twistLimitDegrees=5}
 {bone1Name=belt_mid_02_L, bone2Name=belt_mid_01_L, swing1LimitDegrees=0, swing2LimitDegrees=45, twistLimitDegrees=5}
 {bone1Name=belt_mid_03_L, bone2Name=belt_mid_02_L, swing1LimitDegrees=25, swing2LimitDegrees=45, twistLimitDegrees=5}
 {bone1Name=belt_mid_01_R, bone2Name=belt_mid_MSTR_R, swing1LimitDegrees=0, swing2LimitDegrees=45, twistLimitDegrees=5}
 {bone1Name=belt_mid_02_R, bone2Name=belt_mid_01_R, swing1LimitDegrees=0, swing2LimitDegrees=45, twistLimitDegrees=5}
 {bone1Name=belt_mid_03_R, bone2Name=belt_mid_02_R, swing1LimitDegrees=25, swing2LimitDegrees=45, twistLimitDegrees=5}
 {
  bone1Name=thigh_r
  bone2Name=pelvis
  swing1LimitDegrees=45
  swing1Motion=Free
  swing2LimitDegrees=45
  swing2Motion=Free
  twistLimitDegrees=45
  twistMotion=Free
 }
 {
  bone1Name=belt_02_L
  bone2Name=belt_mid_02_L
  swing1LimitDegrees=45
  swing1Motion=Free
  swing2LimitDegrees=45
  swing2Motion=Free
  twistLimitDegrees=45
  twistMotion=Free
 }
 {
  bone1Name=belt_03_L
  bone2Name=belt_mid_03_L
  swing1LimitDegrees=45
  swing1Motion=Free
  swing2LimitDegrees=45
  swing2Motion=Free
  twistLimitDegrees=45
  twistMotion=Free
 }
 {
  bone1Name=belt_02_R
  bone2Name=belt_mid_02_R
  swing1LimitDegrees=45
  swing1Motion=Free
  swing2LimitDegrees=45
  swing2Motion=Free
  twistLimitDegrees=45
  twistMotion=Free
 }
 {
  bone1Name=belt_mid_02_R
  bone2Name=belt_mid_02_L
  swing1LimitDegrees=45
  swing1Motion=Free
  swing2LimitDegrees=45
  swing2Motion=Free
  twistLimitDegrees=45
  twistMotion=Free
 }
 {
  bone1Name=belt_mid_03_R
  bone2Name=belt_mid_03_L
  swing1LimitDegrees=45
  swing1Motion=Free
  swing2LimitDegrees=45
  swing2Motion=Free
  twistLimitDegrees=45
  twistMotion=Free
 }
 {
  bone1Name=belt_MSTR_L
  bone2Name=spine_02
  swing1LimitDegrees=45
  swing1Motion=Free
  swing2LimitDegrees=45
  swing2Motion=Free
  twistLimitDegrees=45
  twistMotion=Free
 }
 {
  bone1Name=belt_mid_MSTR_L
  bone2Name=spine_02
  swing1LimitDegrees=45
  swing1Motion=Free
  swing2LimitDegrees=45
  swing2Motion=Free
  twistLimitDegrees=45
  twistMotion=Free
 }
 {
  bone1Name=belt_mid_MSTR_R
  bone2Name=spine_02
  swing1LimitDegrees=45
  swing1Motion=Free
  swing2LimitDegrees=45
  swing2Motion=Free
  twistLimitDegrees=45
  twistMotion=Free
 }
 {
  bone1Name=belt_MSTR_R
  bone2Name=spine_02
  swing1LimitDegrees=45
  swing1Motion=Free
  swing2LimitDegrees=45
  swing2Motion=Free
  twistLimitDegrees=45
  twistMotion=Free
 }
 {
  bone1Name=spine_02
  bone2Name=spine_01
  swing1LimitDegrees=45
  swing1Motion=Free
  swing2LimitDegrees=45
  swing2Motion=Free
  twistLimitDegrees=45
  twistMotion=Free
 }
]
