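# Post-process-style cosmetic AnimBP for the MistRaven Peck female body, posing and simulating its belt bones.
# Copies the attached parent's pose, runs a Control Rig and caches the result.
# A DBNO-carried bool selects either hand-tuned additive belt rotations or offsets feeding a RigidBody simulation relative to spine_01.
# The event graph computes right-thigh yaw over interpolated crouch-move ranges and a skydive-dependent gravity vector, with no visible links to AnimGraph.
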AnimBlueprint /BRCosmetics/Characters/Player/Female/Medium/Bodies/F_MED_MistRaven_Peck/Meshes/F_MED_MistRaven_Peck_AnimBP
target_skeleton=&/F_MED_MistRaven_Peck_Skeleton
graphs [
 AnimGraph {nodes [
   n1 = "Transform (Modify) Bone - Bone: belt_MSTR_R"!(ComponentPose=Local To Component(LocalPose=Use cached pose 'Start_Pose'()), Alpha=1, Translation="0,0,0", Rotation="-0.047,-20.181,0.069") -> (Pose) {
    rotationMode=BMM_Additive
    rotationSpace=BCS_BoneSpace
    alphaBoolBlend {blendOption=Linear}
    boneToModify {boneName=belt_MSTR_R}
    rotation=<rot -0.0469 -20.1813 0.069>
   }
   n2 = "Transform (Modify) Bone - Bone: belt_mid_MSTR_R"!(ComponentPose=@n1, Alpha=1, Translation="0,0,0", Rotation="27.127,145.125,157.836") -> (Pose) {
    rotationMode=BMM_Additive
    rotationSpace=BCS_BoneSpace
    alphaBoolBlend {blendOption=Linear}
    boneToModify {boneName=belt_mid_MSTR_R}
    rotation=<rot 27.1268 145.1249 157.836>
   }
   n3 = "Transform (Modify) Bone - Bone: belt_MSTR_L"!(ComponentPose=@n2, Alpha=1, Translation="0,0,0", Rotation="37.759,74.454,97.975") -> (Pose) {
    rotationMode=BMM_Additive
    rotationSpace=BCS_BoneSpace
    alphaBoolBlend {blendOption=Linear}
    boneToModify {boneName=belt_MSTR_L}
    rotation=<rot 37.759 74.454 97.9751>
   }
   n4 = "Transform (Modify) Bone - Bone: belt_mid_MSTR_L"!(ComponentPose=@n3, Alpha=1, Translation="0,0,0", Rotation="-8.896,14.930,47.389") -> (Pose) {
    rotationMode=BMM_Additive
    rotationSpace=BCS_BoneSpace
    alphaBoolBlend {blendOption=Linear}
    boneToModify {boneName=belt_mid_MSTR_L}
    rotation=<rot -8.8959 14.9298 47.3891>
   }
   n5 = RigidBody!(ComponentPose=@n4, Alpha, OverrideWorldGravity:Vector="0,0,0") -> (Pose) {
    bOverrideWorldGravity=true
    overridePhysicsAsset=&/F_MED_MistRaven_Peck_Physics
    simulationSpace=BaseBoneSpace
    alphaBoolBlend {blendOption=Linear}
    baseBoneRef {boneName=spine_01}
    componentAppliedLinearAccClamp=<2500 2500 2500>
   }
   n6 = "Transform (Modify) Bone - Bone: belt_MSTR_R"(ComponentPose=Local To Component(LocalPose=Use cached pose 'Start_Pose'()), Alpha=1, Translation="0,0,0", Rotation="-30,25,7") {
    rotationMode=BMM_Additive
    rotationSpace=BCS_BoneSpace
    alphaBoolBlend {blendOption=Linear}
    boneToModify {boneName=belt_MSTR_R}
    rotation=<rot -30 25 7>
   }
   n7 = "Transform (Modify) Bone - Bone: belt_02_R"!(ComponentPose=@n6, Alpha=1, Translation="0,0,0", Rotation="-51,29,-66") -> (Pose) {
    rotationMode=BMM_Additive
    rotationSpace=BCS_BoneSpace
    alphaBoolBlend {blendOption=Linear}
    boneToModify {boneName=belt_02_R}
    rotation=<rot -51 29 -66>
   }
   n8 = "Transform (Modify) Bone - Bone: belt_03_R"!(ComponentPose=@n7, Alpha=1, Translation="0,0,0", Rotation="-12,0,0") -> (Pose) {
    rotationMode=BMM_Additive
    rotationSpace=BCS_BoneSpace
    alphaBoolBlend {blendOption=Linear}
    boneToModify {boneName=belt_03_R}
    rotation {pitch=-12}
   }
   n9 = "Transform (Modify) Bone - Bone: belt_mid_MSTR_R"(ComponentPose=@n8, Alpha=1, Translation="0,0,0", Rotation="-23,38,15") {
    rotationMode=BMM_Additive
    rotationSpace=BCS_BoneSpace
    alphaBoolBlend {blendOption=Linear}
    boneToModify {boneName=belt_mid_MSTR_R}
    rotation=<rot -23 38 15>
   }
   n10 = "Transform (Modify) Bone - Bone: belt_mid_02_R"!(ComponentPose=@n9, Alpha=1, Translation="0,0,0", Rotation="-53,37,-72") -> (Pose) {
    rotationMode=BMM_Additive
    rotationSpace=BCS_BoneSpace
    alphaBoolBlend {blendOption=Linear}
    boneToModify {boneName=belt_mid_02_R}
    rotation=<rot -53 37 -72>
   }
   n11 = "Transform (Modify) Bone - Bone: belt_mid_03_R"!(ComponentPose=@n10, Alpha=1, Translation="0,0,0", Rotation="-14,0,0") -> (Pose) {
    rotationMode=BMM_Additive
    rotationSpace=BCS_BoneSpace
    alphaBoolBlend {blendOption=Linear}
    boneToModify {boneName=belt_mid_03_R}
    rotation {pitch=-14}
   }
   n12 = "Transform (Modify) Bone - Bone: belt_mid_MSTR_L"(ComponentPose=@n11, Alpha=1, Translation="0,0,0", Rotation="-18,57,34") {
    rotationMode=BMM_Additive
    rotationSpace=BCS_BoneSpace
    alphaBoolBlend {blendOption=Linear}
    boneToModify {boneName=belt_mid_MSTR_L}
    rotation=<rot -18 57 34>
   }
   n13 = "Transform (Modify) Bone - Bone: belt_mid_02_L"!(ComponentPose=@n12, Alpha=1, Translation="0,0,0", Rotation="-55,63,-93") -> (Pose) {
    rotationMode=BMM_Additive
    rotationSpace=BCS_BoneSpace
    alphaBoolBlend {blendOption=Linear}
    boneToModify {boneName=belt_mid_02_L}
    rotation=<rot -55 63 -93>
   }
   n14 = "Transform (Modify) Bone - Bone: belt_mid_03_L"!(ComponentPose=@n13, Alpha=1, Translation="0,0,0", Rotation="-36,0,0") -> (Pose) {
    rotationMode=BMM_Additive
    rotationSpace=BCS_BoneSpace
    alphaBoolBlend {blendOption=Linear}
    boneToModify {boneName=belt_mid_03_L}
    rotation {pitch=-36}
   }
   n15 = "Transform (Modify) Bone - Bone: belt_MSTR_L"(ComponentPose=@n14, Alpha=1, Translation="0,0,0", Rotation="0,68,56") {
    rotationMode=BMM_Additive
    rotationSpace=BCS_BoneSpace
    alphaBoolBlend {blendOption=Linear}
    boneToModify {boneName=belt_MSTR_L}
    rotation {roll=56, yaw=68}
   }
   n16 = "Transform (Modify) Bone - Bone: belt_02_L"!(ComponentPose=@n15, Alpha=1, Translation="0,0,0", Rotation="-53,47,-71") -> (Pose) {
    rotationMode=BMM_Additive
    rotationSpace=BCS_BoneSpace
    alphaBoolBlend {blendOption=Linear}
    boneToModify {boneName=belt_02_L}
    rotation=<rot -53 47 -71>
   }
   n17 = "Transform (Modify) Bone - Bone: belt_03_L"!(ComponentPose=@n16, Alpha=1, Translation="0,0,0", Rotation="-53,0,1") -> (Pose) {
    rotationMode=BMM_Additive
    rotationSpace=BCS_BoneSpace
    alphaBoolBlend {blendOption=Linear}
    boneToModify {boneName=belt_03_L}
    rotation {pitch=-53, roll=1}
   }
   Output Pose(Result=Blend Poses by bool(BlendPose_0=Component To Local(ComponentPose=@n17), BlendPose_1=Component To Local(ComponentPose=@n5), bActiveValue=Get bIsDBNOCarried()) {blendType=Linear}) {layerGroup=None}
   n18 = Control Rig(Source=Copy Pose From Mesh() {bUseAttachedParent=true}) {
    alphaBoolBlend {blendOption=Linear}
    controlRigAssetReference {blueprintRigClass=&/F_MED_MistRaven_Peck_ControlRig.F_MED_MistRaven_Peck_ControlRig_C}
   }
   Save cached pose 'Start_Pose'(Pose=@n18) {cacheName=Start_Pose}
  ]}
 EventGraph {nodes [
   n1 = Event BlueprintUpdateAnimation()
   > Set Delta Time X(Delta Time X=@n1.DeltaTimeX)
   n2 = AND Boolean(A=Get bIsCrouching(), B=Get bIsMoving2D())
   > Set In Range A(In Range A:double=FInterpTo(Current=Get In Range A() -> (In Range A:double), Target=SelectFloat(A=-20, B=-45, bPickA=@n2), DeltaTime=Get Delta Time X(), InterpSpeed=5)) -> (Output_Get:double)
   > Set In Range B(In Range B:double=FInterpTo(Current=Get In Range B() -> (In Range B:double), Target=SelectFloat(A=-60, B=-110, bPickA=@n2), DeltaTime=Get Delta Time X(), InterpSpeed=5)) -> (Output_Get:double)
   n3 = GetDeltaTransformFromRefPose!(self=GetOwnerSkeletalMeshComponent(), BoneName=thigh_r, BaseName=pelvis) -> (ReturnValue_Location:Vector, ReturnValue_Rotation:Rotator, ReturnValue_Scale:Vector)
   > Set thigh r yaw(thigh r yaw:double=MapRangeClamped(Value=BreakRotator(InRot=@n3.ReturnValue_Rotation).Yaw, InRangeA=Get In Range A(), InRangeB=Get In Range B())) -> (Output_Get:double)
   n4 = Get bIsSkydiveFloating()
   n5 = Get bIsSkydiveDiveMode()
   > Set gravity overrride(gravity overrride:Vector=SelectVector(A="0,0,500", B="0,0,-980", bPickA=OR Boolean(A=@n4, B=@n5))) -> (Output_Get:Vector)
  ]}
]
variables [
 Delta Time X {type=number}
 gravity overrride {type=Vector}
 In Range A {type=number}
 In Range B {type=number}
 thigh r yaw {type=number}
]
