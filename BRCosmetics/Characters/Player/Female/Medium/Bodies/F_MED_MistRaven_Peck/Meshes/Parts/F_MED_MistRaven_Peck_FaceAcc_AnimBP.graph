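# Post-copy AnimBP for the MistRaven Peck face-accessory part, adding secondary motion to hair, bangs and strands.
# Two AnimDynamics strand chains on the parent pose, one sprung with rotation offsets, blend by bool before shared bang and hair dynamics.
# A cached result gets right-side constraint, look-at and offset corrections through a per-bone blend.
# Head and right-arm bone deltas weight that correction, zeroed during skydiving, emotes and melee.
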AnimBlueprint /BRCosmetics/Characters/Player/Female/Medium/Bodies/F_MED_MistRaven_Peck/Meshes/Parts/F_MED_MistRaven_Peck_FaceAcc_AnimBP
target_skeleton=&/F_MED_MistRaven_Peck_FaceAcc_Skeleton
graphs [
 AnimGraph {nodes [
   n1 = "Constraint - dyn_strand_end_r (PPP)"!(
    ComponentPose=Local To Component(LocalPose=Use cached pose 'MainDynamicCache'())
    Alpha=1
    ConstraintWeights_0=0.934
    ConstraintWeights_1:float=0.027
    ConstraintWeights_2:float=0.039
   ) -> (Pose) {
    alphaBoolBlend {blendOption=Linear}
    boneToModify {boneName=dyn_strand_end_r}
    constraintSetup [{targetBone {boneName=clavicle_r}}, {targetBone {boneName=spine_04}}, {targetBone {boneName=spine_05}}]
    constraintWeights [1,1,1]
   }
   n2 = Look At - Bone: dyn_strand_00_r(ComponentPose=@n1, Alpha=0.5) {
    alphaBoolBlend {blendOption=Linear}
    boneToModify {boneName=dyn_strand_00_r}
    lookAtLocation=<-2.0031 -9.2207 -2.5742>
    lookAtTarget {boneReference {boneName=dyn_strand_end_r}}
    lookAt_Axis {axis=<-1 0>}
    lookUp_Axis {axis {x=0, z=1}}
   }
   n3 = Look At - Bone: dyn_strand_01_r(ComponentPose=@n2, Alpha=0.5) {
    alphaBoolBlend {blendOption=Linear}
    boneToModify {boneName=dyn_strand_01_r}
    lookAtLocation=<2.7738 -7.0867 2.9156>
    lookAtTarget {boneReference {boneName=dyn_strand_end_r}}
    lookAt_Axis {axis=<-1 0>}
    lookUp_Axis {axis {x=0, z=1}}
   }
   n4 = Look At - Bone: dyn_strand_02_r(ComponentPose=@n3, Alpha=0.3) {
    alphaBoolBlend {blendOption=Linear}
    boneToModify {boneName=dyn_strand_02_r}
    lookAtLocation=<0.7326 -12.6471 3.6138>
    lookAtTarget {boneReference {boneName=dyn_strand_end_r}}
    lookAt_Axis {axis=<-1 0>}
    lookUp_Axis {axis {x=0, z=1}}
   }
   n5 = Look At - Bone: dyn_bang_strand_00_r(ComponentPose=@n4, Alpha=0.5) {
    alphaBoolBlend {blendOption=Linear}
    boneToModify {boneName=dyn_bang_strand_00_r}
    lookAtLocation=<1.0245 -10.9365 -0.3291>
    lookAtTarget {boneReference {boneName=dyn_strand_end_r}}
    lookAt_Axis {axis=<-1 0>}
    lookUp_Axis {axis {x=0, z=1}}
   }
   n6 = "Transform (Modify) Bone - Bone: dyn_hair_00_c"(ComponentPose=@n5, Alpha=1, Translation="-0.646,-0.015,-0.172", Rotation="10.595,1.437,0.850", Scale="1,1,1") {
    rotationMode=BMM_Additive
    rotationSpace=BCS_BoneSpace
    translationMode=BMM_Additive
    translationSpace=BCS_BoneSpace
    alphaBoolBlend {blendOption=Linear}
    boneToModify {boneName=dyn_hair_00_c}
    rotation=<rot 10.595 1.437 0.85>
    translation=<-0.646 -0.015 -0.172>
   }
   n7 = "Transform (Modify) Bone - Bone: dyn_hair_00_r"(ComponentPose=@n6, Alpha=1, Translation="-1.454,0.254,0.379", Rotation="9.004,-3.474,7.811", Scale="1,1,1") {
    rotationMode=BMM_Additive
    rotationSpace=BCS_BoneSpace
    translationMode=BMM_Additive
    translationSpace=BCS_BoneSpace
    alphaBoolBlend {blendOption=Linear}
    boneToModify {boneName=dyn_hair_00_r}
    rotation=<rot 9.004 -3.474 7.811>
    translation=<-1.454 0.254 0.379>
   }
   n8 = Layered blend per bone(
    BasePose=Use cached pose 'MainDynamicCache'()
    BlendPoses_0=Component To Local(ComponentPose=@n7)
    BlendWeights_0=Get Targeting correction() -> (Targeting correction:double)
   ) {layerSetup [{branchFilters [
       {boneName=dyn_strand_00_r}
       {boneName=dyn_strand_01_r}
       {boneName=dyn_strand_02_r}
       {boneName=dyn_bang_strand_00_r}
       {boneName=dyn_hair_00_r}
       {boneName=dyn_hair_00_c}
      ]}]}
   Output Pose(Result=@n8) {layerGroup=None}
   n9 = Anim Dynamics - Bone: dyn_strand_00_r(ComponentPose=Local To Component(LocalPose=Copy Pose From Mesh() {bUseAttachedParent=true}), Alpha=1, ExternalForce="0,0,0") {
    angularDampingOverride=0.99
    angularSpringConstant=6000
    bAngularSpring=true
    bOverrideAngularDamping=true
    bUseGravityOverride=true
    alphaBoolBlend {blendOption=Linear}
    boundBone {boneName=dyn_strand_00_r}
    componentAppliedLinearAccClamp=<3000 3000 3000>
    componentLinearAccScale=<0.4 0.4 0.4>
    componentLinearVelScale=<0.6 0.6 0.6>
    gravityOverride {z=-980}
    physicsBodyDefinitions [{
      boundBone {boneName=dyn_strand_00_r}
      boxExtents=<20 20 20>
      constraintSetup {angularLimitsMax=<0 4 15>, angularLimitsMin=<0 -20 -15>, angularTarget=<1 0 0>}
      localJointOffset=<-0.8 0 0>
     }]
   }
   n10 = Anim Dynamics - Bone: dyn_strand_01_r(ComponentPose=@n9, Alpha=1, ExternalForce="0,0,0") {
    angularDampingOverride=0.99
    angularSpringConstant=6000
    bAngularSpring=true
    bOverrideAngularDamping=true
    bUseGravityOverride=true
    alphaBoolBlend {blendOption=Linear}
    boundBone {boneName=dyn_strand_01_r}
    componentAppliedLinearAccClamp=<3000 3000 3000>
    componentLinearAccScale=<0.4 0.4 0.4>
    componentLinearVelScale=<0.6 0.6 0.6>
    gravityOverride {z=-980}
    physicsBodyDefinitions [{
      boundBone {boneName=dyn_strand_01_r}
      boxExtents=<20 20 20>
      constraintSetup {angularLimitsMax=<0 6 18>, angularLimitsMin=<0 -20 -15>, angularTarget=<1 0 0>}
      localJointOffset=<-0.8 0 0>
     }]
   }
   n11 = Anim Dynamics - Bone: dyn_strand_02_r(ComponentPose=@n10, Alpha=1, ExternalForce="0,0,0") {
    angularDampingOverride=0.99
    angularSpringConstant=6000
    bAngularSpring=true
    bOverrideAngularDamping=true
    bUseGravityOverride=true
    bUseSphericalLimits=true
    alphaBoolBlend {blendOption=Linear}
    boundBone {boneName=dyn_strand_02_r}
    componentAppliedLinearAccClamp=<3000 3000 3000>
    componentLinearAccScale=<0.4 0.4 0.4>
    componentLinearVelScale=<0.6 0.6 0.6>
    gravityOverride {z=-980}
    physicsBodyDefinitions [{
      collisionType=CustomSphere
      sphereCollisionRadius=1.2
      boundBone {boneName=dyn_strand_02_r}
      boxExtents=<40 40 40>
      constraintSetup {angularLimitsMax=<0 15 20>, angularLimitsMin=<0 -20 -30>, angularTarget=<1 0 0>}
      localJointOffset=<-8 0.5 0>
     }]
    sphericalLimits [{limitRadius=10, drivingBone {boneName=neck_01}, sphereLocalOffset=<1.6293 -3.3396 -2.7914>}]
   }
   n12 = Anim Dynamics - Bone: dyn_bang_strand_00_r(ComponentPose=@n11, Alpha=1, ExternalForce="0,0,0") {
    angularDampingOverride=0.99
    angularSpringConstant=6000
    bAngularSpring=true
    bOverrideAngularDamping=true
    bUseGravityOverride=true
    bUsePlanarLimit=false
    alphaBoolBlend {blendOption=Linear}
    boundBone {boneName=dyn_bang_strand_00_r}
    componentAppliedLinearAccClamp=<3000 3000 3000>
    componentLinearAccScale=<0.4 0.4 0.4>
    componentLinearVelScale=<0.6 0.6 0.6>
    gravityOverride {z=-980}
    physicsBodyDefinitions [{
      boundBone {boneName=dyn_bang_strand_00_r}
      boxExtents=<15 15 15>
      constraintSetup {angularLimitsMax=<0 3 10>, angularLimitsMin=<0 -20 -15>, angularTarget=<1 0 0>}
      localJointOffset=<-1.5 0 0>
     }]
   }
   n13 = Anim Dynamics - Bone: dyn_bang_strand_01_r(ComponentPose=@n12, Alpha=1, ExternalForce="0,0,0") {
    angularDampingOverride=0.99
    angularSpringConstant=6000
    bAngularSpring=true
    bOverrideAngularDamping=true
    bUseGravityOverride=true
    bUsePlanarLimit=false
    alphaBoolBlend {blendOption=Linear}
    boundBone {boneName=dyn_bang_strand_01_r}
    componentAppliedLinearAccClamp=<3000 3000 3000>
    componentLinearAccScale=<0.4 0.4 0.4>
    componentLinearVelScale=<0.6 0.6 0.6>
    gravityOverride {z=-980}
    physicsBodyDefinitions [{
      boundBone {boneName=dyn_bang_strand_01_r}
      boxExtents=<25 25 25>
      constraintSetup {angularLimitsMax=<0 0 25>, angularLimitsMin=<0 -20 -20>, angularTarget=<1 0 0>}
      localJointOffset=<-3 0 0>
     }]
   }
   n14 = Anim Dynamics - Bone: dyn_strand_00_l(ComponentPose=@n13, Alpha=1, ExternalForce="0,0,0") {
    angularDampingOverride=0.99
    angularSpringConstant=6000
    bAngularSpring=true
    bOverrideAngularDamping=true
    bUseGravityOverride=true
    alphaBoolBlend {blendOption=Linear}
    boundBone {boneName=dyn_strand_00_l}
    componentAppliedLinearAccClamp=<3000 3000 3000>
    componentLinearAccScale=<0.4 0.4 0.4>
    componentLinearVelScale=<0.6 0.6 0.6>
    gravityOverride {z=-980}
    physicsBodyDefinitions [{
      boundBone {boneName=dyn_strand_00_l}
      boxExtents=<20 20 20>
      constraintSetup {angularLimitsMax=<0 2 15>, angularLimitsMin=<0 -20 -15>, angularTarget=<1 0 0>}
      localJointOffset=<0.8 0 0>
     }]
   }
   n15 = Anim Dynamics - Bone: dyn_strand_01_l(ComponentPose=@n14, Alpha=1, ExternalForce="0,0,0") {
    angularDampingOverride=0.99
    angularSpringConstant=6000
    bAngularSpring=true
    bOverrideAngularDamping=true
    bUseGravityOverride=true
    alphaBoolBlend {blendOption=Linear}
    boundBone {boneName=dyn_strand_01_l}
    componentAppliedLinearAccClamp=<3000 3000 3000>
    componentLinearAccScale=<0.4 0.4 0.4>
    componentLinearVelScale=<0.6 0.6 0.6>
    gravityOverride {z=-980}
    physicsBodyDefinitions [{
      boundBone {boneName=dyn_strand_01_l}
      boxExtents=<20 20 20>
      constraintSetup {angularLimitsMax=<0 2 15>, angularLimitsMin=<0 -20 -10>, angularTarget=<1 0 0>}
      localJointOffset=<0.8 0 0>
     }]
   }
   n16 = Anim Dynamics - Bone: dyn_strand_02_l(ComponentPose=@n15, Alpha=1, ExternalForce="0,0,0") {
    angularDampingOverride=0.99
    angularSpringConstant=6000
    bAngularSpring=true
    bOverrideAngularDamping=true
    bUseGravityOverride=true
    bUsePlanarLimit=false
    bUseSphericalLimits=true
    alphaBoolBlend {blendOption=Linear}
    boundBone {boneName=dyn_strand_02_l}
    componentAppliedLinearAccClamp=<3000 3000 3000>
    componentLinearAccScale=<0.4 0.4 0.4>
    componentLinearVelScale=<0.6 0.6 0.6>
    gravityOverride {z=-980}
    physicsBodyDefinitions [{
      collisionType=CustomSphere
      sphereCollisionRadius=1.2
      boundBone {boneName=dyn_strand_02_l}
      boxExtents=<40 40 40>
      constraintSetup {angularLimitsMax=<0 15 20>, angularLimitsMin=<0 -20 -30>, angularTarget=<1 0 0>}
      localJointOffset=<8 -0.5 0>
     }]
    planarLimits [{drivingBone {boneName=head}, planeTransform {location=<0.154 0.2998 -6.7256>, rotation=<rot -12.8463 2.3495 164.7376>}}]
    sphericalLimits [{limitRadius=10, drivingBone {boneName=neck_01}, sphereLocalOffset=<1.6293 -3.1988 3.7092>}]
   }
   n17 = Anim Dynamics - Bone: dyn_bang_strand_00_l(ComponentPose=@n16, Alpha=1, ExternalForce="0,0,0") {
    angularDampingOverride=0.99
    angularSpringConstant=6000
    bAngularSpring=true
    bOverrideAngularDamping=true
    bUseGravityOverride=true
    bUsePlanarLimit=false
    alphaBoolBlend {blendOption=Linear}
    boundBone {boneName=dyn_bang_strand_00_l}
    componentAppliedLinearAccClamp=<3000 3000 3000>
    componentLinearAccScale=<0.4 0.4 0.4>
    componentLinearVelScale=<0.6 0.6 0.6>
    gravityOverride {z=-980}
    physicsBodyDefinitions [{
      boundBone {boneName=dyn_bang_strand_00_l}
      boxExtents=<15 15 15>
      constraintSetup {angularLimitsMax=<0 6 10>, angularLimitsMin=<0 -20 -15>, angularTarget=<1 0 0>}
      localJointOffset=<1.5 0 0>
     }]
   }
   n18 = Anim Dynamics - Bone: dyn_bang_strand_01_l(ComponentPose=@n17, Alpha=1, ExternalForce="0,0,0") {
    angularDampingOverride=0.99
    angularSpringConstant=6000
    bAngularSpring=true
    bOverrideAngularDamping=true
    bUseGravityOverride=true
    bUsePlanarLimit=false
    bUseSphericalLimits=true
    alphaBoolBlend {blendOption=Linear}
    boundBone {boneName=dyn_bang_strand_01_l}
    componentAppliedLinearAccClamp=<3000 3000 3000>
    componentLinearAccScale=<0.4 0.4 0.4>
    componentLinearVelScale=<0.6 0.6 0.6>
    gravityOverride {z=-980}
    physicsBodyDefinitions [{
      collisionType=CustomSphere
      sphereCollisionRadius=1
      boundBone {boneName=dyn_bang_strand_01_l}
      boxExtents=<25 25 25>
      constraintSetup {angularLimitsMax=<0 0 25>, angularLimitsMin=<0 -20 -20>, angularTarget=<1 0 0>}
      localJointOffset=<3 0 0>
     }]
    sphericalLimits [{limitRadius=15, drivingBone {boneName=head}, sphereLocalOffset=<1.3026 0.3838 8.3212>}]
   }
   n19 = "Transform (Modify) Bone - Bone: dyn_strand_start_r"!(ComponentPose=@n18, bAlphaBoolEnabled:bool=True, Translation="0,0,0", Rotation="13.719,-8.712,-15.191", Scale="1,1,1") -> (Pose) {
    alphaInputType=Bool
    rotationMode=BMM_Additive
    rotationSpace=BCS_BoneSpace
    alphaBoolBlend {blendInTime=0.8, blendOption=Linear, blendOutTime=0.8}
    boneToModify {boneName=dyn_strand_start_r}
    rotation=<rot 13.719 -8.712 -15.191>
   }
   n20 = "Transform (Modify) Bone - Bone: dyn_bang_strand_00_r"!(ComponentPose=@n19, bAlphaBoolEnabled:bool=True, Translation="0,0,0", Rotation="6.367,-26.485,-0.617", Scale="1,1,1") -> (Pose) {
    alphaInputType=Bool
    rotationMode=BMM_Additive
    rotationSpace=BCS_BoneSpace
    alphaBoolBlend {blendInTime=0.8, blendOption=Linear, blendOutTime=0.8}
    boneToModify {boneName=dyn_bang_strand_00_r}
    rotation=<rot 6.367 -26.485 -0.617>
   }
   n21 = "Transform (Modify) Bone - Bone: dyn_strand_start_l"!(ComponentPose=@n20, bAlphaBoolEnabled:bool=True, Translation="0,0,0", Rotation="10.237,-3.321,-13.163", Scale="1,1,1") -> (Pose) {
    alphaInputType=Bool
    rotationMode=BMM_Additive
    rotationSpace=BCS_BoneSpace
    alphaBoolBlend {blendInTime=0.8, blendOption=Linear, blendOutTime=0.8}
    boneToModify {boneName=dyn_strand_start_l}
    rotation=<rot 10.237 -3.321 -13.163>
   }
   n22 = "Transform (Modify) Bone - Bone: dyn_bang_strand_00_l"!(ComponentPose=@n21, bAlphaBoolEnabled:bool=True, Translation="0,0,0", Rotation="5.645,-21.008,0.316", Scale="1,1,1") -> (Pose) {
    alphaInputType=Bool
    rotationMode=BMM_Additive
    rotationSpace=BCS_BoneSpace
    alphaBoolBlend {blendInTime=0.8, blendOption=Linear, blendOutTime=0.8}
    boneToModify {boneName=dyn_bang_strand_00_l}
    rotation=<rot 5.645 -21.008 0.316>
   }
   n23 = Component To Local(ComponentPose=@n22)
   n24 = Anim Dynamics - Bone: dyn_strand_00_r(ComponentPose=Local To Component(LocalPose=Copy Pose From Mesh() {bUseAttachedParent=true}), Alpha=1, ExternalForce="0,0,0") {
    angularDampingOverride=0.99
    angularSpringConstant=10000
    bOverrideAngularDamping=true
    bUseGravityOverride=true
    alphaBoolBlend {blendOption=Linear}
    boundBone {boneName=dyn_strand_00_r}
    componentAppliedLinearAccClamp=<3000 3000 3000>
    componentLinearAccScale=<0.4 0.4 0.4>
    componentLinearVelScale=<0.6 0.6 0.6>
    gravityOverride {z=-980}
    physicsBodyDefinitions [{
      boundBone {boneName=dyn_strand_00_r}
      boxExtents=<12 12 10>
      constraintSetup {angularLimitsMax=<0 0 15>, angularLimitsMin=<0 -20 -9>, angularTarget=<1 0 0>}
      localJointOffset=<-0.8 0 0>
     }]
   }
   n25 = Anim Dynamics - Bone: dyn_strand_01_r(ComponentPose=@n24, Alpha=1, ExternalForce="0,0,0") {
    angularDampingOverride=0.99
    angularSpringConstant=10000
    bOverrideAngularDamping=true
    bUseGravityOverride=true
    alphaBoolBlend {blendOption=Linear}
    boundBone {boneName=dyn_strand_01_r}
    componentAppliedLinearAccClamp=<3000 3000 3000>
    componentLinearAccScale=<0.4 0.4 0.4>
    componentLinearVelScale=<0.6 0.6 0.6>
    gravityOverride {z=-980}
    physicsBodyDefinitions [{
      boundBone {boneName=dyn_strand_01_r}
      boxExtents=<10 10 10>
      constraintSetup {angularLimitsMax=<0 6 18>, angularLimitsMin=<0 -20 -13>, angularTarget=<1 0 0>}
      localJointOffset=<-0.8 0 0>
     }]
   }
   n26 = Anim Dynamics - Bone: dyn_strand_02_r(ComponentPose=@n25, Alpha=1, ExternalForce="0,0,0") {
    angularDampingOverride=0.99
    angularSpringConstant=10000
    bOverrideAngularDamping=true
    bUseGravityOverride=true
    bUseSphericalLimits=true
    alphaBoolBlend {blendOption=Linear}
    boundBone {boneName=dyn_strand_02_r}
    componentAppliedLinearAccClamp=<3000 3000 3000>
    componentLinearAccScale=<0.4 0.4 0.4>
    componentLinearVelScale=<0.6 0.6 0.6>
    gravityOverride {z=-980}
    physicsBodyDefinitions [{
      collisionType=CustomSphere
      sphereCollisionRadius=1.2
      boundBone {boneName=dyn_strand_02_r}
      boxExtents=<40 40 40>
      constraintSetup {angularLimitsMax=<0 15 20>, angularLimitsMin=<0 -20 -30>, angularTarget=<1 0 0>}
      localJointOffset=<-8 0.5 0>
     }]
    sphericalLimits [{limitRadius=10, drivingBone {boneName=neck_01}, sphereLocalOffset=<1.6293 -3.3396 -2.7914>}]
   }
   n27 = Anim Dynamics - Bone: dyn_bang_strand_00_r(ComponentPose=@n26, Alpha=1, ExternalForce="0,0,0") {
    angularDampingOverride=0.99
    angularSpringConstant=10000
    bOverrideAngularDamping=true
    bUseGravityOverride=true
    bUsePlanarLimit=false
    alphaBoolBlend {blendOption=Linear}
    boundBone {boneName=dyn_bang_strand_00_r}
    componentAppliedLinearAccClamp=<3000 3000 3000>
    componentLinearAccScale=<0.4 0.4 0.4>
    componentLinearVelScale=<0.6 0.6 0.6>
    gravityOverride {z=-980}
    physicsBodyDefinitions [{
      boundBone {boneName=dyn_bang_strand_00_r}
      boxExtents=<15 15 15>
      constraintSetup {angularLimitsMax=<0 3 10>, angularLimitsMin=<0 -20 -15>, angularTarget=<1 0 0>}
      localJointOffset=<-1.5 0 0>
     }]
   }
   n28 = Anim Dynamics - Bone: dyn_bang_strand_01_r(ComponentPose=@n27, Alpha=1, ExternalForce="0,0,0") {
    angularDampingOverride=0.99
    angularSpringConstant=10000
    bOverrideAngularDamping=true
    bUseGravityOverride=true
    bUsePlanarLimit=false
    alphaBoolBlend {blendOption=Linear}
    boundBone {boneName=dyn_bang_strand_01_r}
    componentAppliedLinearAccClamp=<3000 3000 3000>
    componentLinearAccScale=<0.4 0.4 0.4>
    componentLinearVelScale=<0.6 0.6 0.6>
    gravityOverride {z=-980}
    physicsBodyDefinitions [{
      boundBone {boneName=dyn_bang_strand_01_r}
      boxExtents=<25 25 25>
      constraintSetup {angularLimitsMax=<0 0 25>, angularLimitsMin=<0 -20 -20>, angularTarget=<1 0 0>}
      localJointOffset=<-3 0 0>
     }]
   }
   n29 = Anim Dynamics - Bone: dyn_strand_00_l(ComponentPose=@n28, Alpha=1, ExternalForce="0,0,0") {
    angularDampingOverride=0.99
    angularSpringConstant=10000
    bOverrideAngularDamping=true
    bUseGravityOverride=true
    alphaBoolBlend {blendOption=Linear}
    boundBone {boneName=dyn_strand_00_l}
    componentAppliedLinearAccClamp=<3000 3000 3000>
    componentLinearAccScale=<0.4 0.4 0.4>
    componentLinearVelScale=<0.6 0.6 0.6>
    gravityOverride {z=-980}
    physicsBodyDefinitions [{
      boundBone {boneName=dyn_strand_00_l}
      boxExtents=<12 12 10>
      constraintSetup {angularLimitsMax=<0 2 15>, angularLimitsMin=<0 -20 -15>, angularTarget=<1 0 0>}
      localJointOffset=<0.8 0 0>
     }]
   }
   n30 = Anim Dynamics - Bone: dyn_strand_01_l(ComponentPose=@n29, Alpha=1, ExternalForce="0,0,0") {
    angularDampingOverride=0.99
    angularSpringConstant=10000
    bOverrideAngularDamping=true
    bUseGravityOverride=true
    alphaBoolBlend {blendOption=Linear}
    boundBone {boneName=dyn_strand_01_l}
    componentAppliedLinearAccClamp=<3000 3000 3000>
    componentLinearAccScale=<0.4 0.4 0.4>
    componentLinearVelScale=<0.6 0.6 0.6>
    gravityOverride {z=-980}
    physicsBodyDefinitions [{
      boundBone {boneName=dyn_strand_01_l}
      boxExtents=<10 10 10>
      constraintSetup {angularLimitsMax=<0 2 15>, angularLimitsMin=<0 -20 -10>, angularTarget=<1 0 0>}
      localJointOffset=<0.8 0 0>
     }]
   }
   n31 = Anim Dynamics - Bone: dyn_strand_02_l(ComponentPose=@n30, Alpha=1, ExternalForce="0,0,0") {
    angularDampingOverride=0.99
    angularSpringConstant=10000
    bOverrideAngularDamping=true
    bUseGravityOverride=true
    bUsePlanarLimit=false
    bUseSphericalLimits=true
    alphaBoolBlend {blendOption=Linear}
    boundBone {boneName=dyn_strand_02_l}
    componentAppliedLinearAccClamp=<3000 3000 3000>
    componentLinearAccScale=<0.4 0.4 0.4>
    componentLinearVelScale=<0.6 0.6 0.6>
    gravityOverride {z=-980}
    physicsBodyDefinitions [{
      collisionType=CustomSphere
      sphereCollisionRadius=1.2
      boundBone {boneName=dyn_strand_02_l}
      boxExtents=<40 40 40>
      constraintSetup {angularLimitsMax=<0 15 20>, angularLimitsMin=<0 -20 -30>, angularTarget=<1 0 0>}
      localJointOffset=<8 -0.5 0>
     }]
    planarLimits [{drivingBone {boneName=head}, planeTransform {location=<0.154 0.2998 -6.7256>, rotation=<rot -12.8463 2.3495 164.7376>}}]
    sphericalLimits [{limitRadius=10, drivingBone {boneName=neck_01}, sphereLocalOffset=<1.6293 -3.1988 3.7092>}]
   }
   n32 = Anim Dynamics - Bone: dyn_bang_strand_00_l(ComponentPose=@n31, Alpha=1, ExternalForce="0,0,0") {
    angularDampingOverride=0.99
    angularSpringConstant=10000
    bOverrideAngularDamping=true
    bUseGravityOverride=true
    bUsePlanarLimit=false
    alphaBoolBlend {blendOption=Linear}
    boundBone {boneName=dyn_bang_strand_00_l}
    componentAppliedLinearAccClamp=<3000 3000 3000>
    componentLinearAccScale=<0.4 0.4 0.4>
    componentLinearVelScale=<0.6 0.6 0.6>
    gravityOverride {z=-980}
    physicsBodyDefinitions [{
      boundBone {boneName=dyn_bang_strand_00_l}
      boxExtents=<15 15 15>
      constraintSetup {angularLimitsMax=<0 2 10>, angularLimitsMin=<0 -20 -15>, angularTarget=<1 0 0>}
      localJointOffset=<1.5 0 0>
     }]
   }
   n33 = Anim Dynamics - Bone: dyn_bang_strand_01_l(ComponentPose=@n32, Alpha=1, ExternalForce="0,0,0") {
    angularDampingOverride=0.99
    angularSpringConstant=10000
    bOverrideAngularDamping=true
    bUseGravityOverride=true
    bUsePlanarLimit=false
    bUseSphericalLimits=true
    alphaBoolBlend {blendOption=Linear}
    boundBone {boneName=dyn_bang_strand_01_l}
    componentAppliedLinearAccClamp=<3000 3000 3000>
    componentLinearAccScale=<0.4 0.4 0.4>
    componentLinearVelScale=<0.6 0.6 0.6>
    gravityOverride {z=-980}
    physicsBodyDefinitions [{
      collisionType=CustomSphere
      sphereCollisionRadius=1
      boundBone {boneName=dyn_bang_strand_01_l}
      boxExtents=<25 25 25>
      constraintSetup {angularLimitsMax=<0 0 25>, angularLimitsMin=<0 -20 -20>, angularTarget=<1 0 0>}
      localJointOffset=<3 0 0>
     }]
    sphericalLimits [{limitRadius=15, drivingBone {boneName=head}, sphereLocalOffset=<1.3026 0.3838 8.2441>}]
   }
   n34 = Component To Local(ComponentPose=@n33)
   n35 = Anim Dynamics - Bone: dyn_bang_l(
    ComponentPose=Local To Component(LocalPose=Blend Poses by bool(BlendPose_0=@n23, BlendPose_1=@n34) {blendType=Linear})
    Alpha=1
    ExternalForce="0,0,0"
   ) {
    angularDampingOverride=0.99
    bOverrideAngularDamping=true
    bUseGravityOverride=true
    bUsePlanarLimit=false
    alphaBoolBlend {blendOption=Linear}
    boundBone {boneName=dyn_bang_l}
    componentAppliedLinearAccClamp=<3000 3000 3000>
    componentLinearAccScale=<0.4 0.4 0.4>
    componentLinearVelScale=<0.6 0.6 0.6>
    gravityOverride {z=-980}
    physicsBodyDefinitions [{
      boundBone {boneName=dyn_bang_l}
      boxExtents=<15 15 15>
      constraintSetup {angularLimitsMax=<0 0 8>, angularLimitsMin=<0 -20 -5>, angularTarget=<0 0 0>}
      localJointOffset=<1.5 -0.1 0>
     }]
   }
   n36 = Anim Dynamics - Bone: dyn_bang_c(ComponentPose=@n35, Alpha=1, ExternalForce="0,0,0") {
    angularDampingOverride=0.99
    bOverrideAngularDamping=true
    bUseGravityOverride=true
    bUsePlanarLimit=false
    alphaBoolBlend {blendOption=Linear}
    boundBone {boneName=dyn_bang_c}
    componentAppliedLinearAccClamp=<3000 3000 3000>
    componentLinearAccScale=<0.4 0.4 0.4>
    componentLinearVelScale=<0.6 0.6 0.6>
    gravityOverride {z=-980}
    physicsBodyDefinitions [{
      boundBone {boneName=dyn_bang_c}
      boxExtents=<15 15 15>
      constraintSetup {angularLimitsMax=<0 0 12>, angularLimitsMin=<0 -20 -10>, angularTarget=<0 0 0>}
      localJointOffset=<1.5 0 0>
     }]
   }
   n37 = Anim Dynamics - Bone: dyn_bang_r(ComponentPose=@n36, Alpha=1, ExternalForce="0,0,0") {
    angularDampingOverride=0.99
    bOverrideAngularDamping=true
    bUseGravityOverride=true
    bUsePlanarLimit=false
    alphaBoolBlend {blendOption=Linear}
    boundBone {boneName=dyn_bang_r}
    componentAppliedLinearAccClamp=<3000 3000 3000>
    componentLinearAccScale=<0.4 0.4 0.4>
    componentLinearVelScale=<0.6 0.6 0.6>
    gravityOverride {z=-980}
    physicsBodyDefinitions [{
      boundBone {boneName=dyn_bang_r}
      boxExtents=<15 15 15>
      constraintSetup {angularLimitsMax=<0 0 15>, angularLimitsMin=<0 -20 -12>, angularTarget=<0 0 0>}
      localJointOffset=<1.5 0.1 0>
     }]
   }
   n38 = Anim Dynamics - Bone: dyn_hair_00_l(ComponentPose=@n37, Alpha=1, ExternalForce="0,0,0") {
    angularDampingOverride=0.9999
    angularSpringConstant=8000
    bAngularSpring=true
    bOverrideAngularDamping=true
    bUseGravityOverride=true
    bUsePlanarLimit=false
    alphaBoolBlend {blendOption=Linear}
    boundBone {boneName=dyn_hair_00_l}
    componentAppliedLinearAccClamp=<3000 3000 3000>
    componentLinearAccScale=<0.4 0.4 0.4>
    componentLinearVelScale=<0.6 0.6 0.6>
    gravityOverride {z=-980}
    physicsBodyDefinitions [{
      boundBone {boneName=dyn_hair_00_l}
      boxExtents=<26 26 26>
      constraintSetup {angularLimitsMax=<0 15 8>, angularLimitsMin=<0 0 -20>, angularTarget=<1 0 0>}
      localJointOffset=<4 2 0>
     }]
   }
   n39 = Anim Dynamics - Bone: dyn_hair_00_c(ComponentPose=@n38, Alpha=1, ExternalForce="0,0,0") {
    angularDampingOverride=0.9999
    angularSpringConstant=8000
    bAngularSpring=true
    bOverrideAngularDamping=true
    bUseGravityOverride=true
    bUsePlanarLimit=false
    alphaBoolBlend {blendOption=Linear}
    boundBone {boneName=dyn_hair_00_c}
    componentAppliedLinearAccClamp=<3000 3000 3000>
    componentLinearAccScale=<0.4 0.4 0.4>
    componentLinearVelScale=<0.6 0.6 0.6>
    gravityOverride {z=-980}
    physicsBodyDefinitions [{
      boundBone {boneName=dyn_hair_00_c}
      boxExtents=<26 26 26>
      constraintSetup {angularLimitsMax=<0 20 10>, angularLimitsMin=<0 0 -10>, angularTarget=<1 0 0>}
      localJointOffset=<4 0 0>
     }]
   }
   n40 = Anim Dynamics - Bone: dyn_hair_00_r(ComponentPose=@n39, Alpha=1, ExternalForce="0,0,0") {
    angularDampingOverride=0.9999
    angularSpringConstant=8000
    bAngularSpring=true
    bOverrideAngularDamping=true
    bUseGravityOverride=true
    bUsePlanarLimit=false
    alphaBoolBlend {blendOption=Linear}
    boundBone {boneName=dyn_hair_00_r}
    componentAppliedLinearAccClamp=<3000 3000 3000>
    componentLinearAccScale=<0.4 0.4 0.4>
    componentLinearVelScale=<0.6 0.6 0.6>
    gravityOverride {z=-980}
    physicsBodyDefinitions [{
      boundBone {boneName=dyn_hair_00_r}
      boxExtents=<26 26 26>
      constraintSetup {angularLimitsMax=<6 15 20>, angularLimitsMin=<-6 0 -6>, angularTarget=<1 0 0>}
      localJointOffset=<4 -2 0>
     }]
   }
   n41 = Anim Dynamics - Bone: dyn_hair_01_l(ComponentPose=@n40, Alpha=1, ExternalForce="0,0,0") {
    angularDampingOverride=0.9999
    angularSpringConstant=6000
    bAngularSpring=true
    bOverrideAngularDamping=true
    bUseGravityOverride=true
    bUsePlanarLimit=false
    alphaBoolBlend {blendOption=Linear}
    boundBone {boneName=dyn_hair_01_l}
    componentAppliedLinearAccClamp=<3000 3000 3000>
    componentLinearAccScale=<0.4 0.4 0.4>
    componentLinearVelScale=<0.6 0.6 0.6>
    gravityOverride {z=-980}
    physicsBodyDefinitions [{
      boundBone {boneName=dyn_hair_01_l}
      boxExtents=<22 22 22>
      constraintSetup {angularLimitsMax=<0 10 8>, angularLimitsMin=<0 -10 -8>, angularTarget=<1 0 0>}
      localJointOffset=<4 0 0>
     }]
   }
   n42 = Anim Dynamics - Bone: dyn_hair_01_side_l(ComponentPose=@n41, Alpha=1, ExternalForce="0,0,0") {
    angularDampingOverride=0.9999
    angularSpringConstant=6000
    bAngularSpring=true
    bOverrideAngularDamping=true
    bUseGravityOverride=true
    bUsePlanarLimit=false
    alphaBoolBlend {blendOption=Linear}
    boundBone {boneName=dyn_hair_01_side_l}
    componentAppliedLinearAccClamp=<3000 3000 3000>
    componentLinearAccScale=<0.4 0.4 0.4>
    componentLinearVelScale=<0.6 0.6 0.6>
    gravityOverride {z=-980}
    physicsBodyDefinitions [{
      boundBone {boneName=dyn_hair_01_side_l}
      boxExtents=<25 25 25>
      constraintSetup {angularLimitsMax=<0 10 8>, angularLimitsMin=<0 -10 -6>, angularTarget=<1 0 0>}
      localJointOffset=<4 0 0>
     }]
   }
   n43 = Anim Dynamics - Bone: dyn_hair_01_c(ComponentPose=@n42, Alpha=1, ExternalForce="0,0,0") {
    angularDampingOverride=0.9999
    angularSpringConstant=6000
    bAngularSpring=true
    bOverrideAngularDamping=true
    bUseGravityOverride=true
    bUsePlanarLimit=false
    alphaBoolBlend {blendOption=Linear}
    boundBone {boneName=dyn_hair_01_c}
    componentAppliedLinearAccClamp=<3000 3000 3000>
    componentLinearAccScale=<0.4 0.4 0.4>
    componentLinearVelScale=<0.6 0.6 0.6>
    gravityOverride {z=-980}
    physicsBodyDefinitions [{
      boundBone {boneName=dyn_hair_01_c}
      boxExtents=<22 22 22>
      constraintSetup {angularLimitsMax=<0 10 8>, angularLimitsMin=<0 -10 -8>, angularTarget=<1 0 0>}
      localJointOffset=<4 0 0>
     }]
   }
   n44 = Anim Dynamics - Bone: dyn_hair_01_r(ComponentPose=@n43, Alpha=1, ExternalForce="0,0,0") {
    angularDampingOverride=0.9999
    angularSpringConstant=6000
    bAngularSpring=true
    bOverrideAngularDamping=true
    bUseGravityOverride=true
    bUsePlanarLimit=false
    alphaBoolBlend {blendOption=Linear}
    boundBone {boneName=dyn_hair_01_r}
    componentAppliedLinearAccClamp=<3000 3000 3000>
    componentLinearAccScale=<0.4 0.4 0.4>
    componentLinearVelScale=<0.6 0.6 0.6>
    gravityOverride {z=-980}
    physicsBodyDefinitions [{
      boundBone {boneName=dyn_hair_01_r}
      boxExtents=<23 23 23>
      constraintSetup {angularLimitsMax=<0 10 8>, angularLimitsMin=<0 -10 -8>, angularTarget=<1 0 0>}
      localJointOffset=<4 0 0>
     }]
   }
   n45 = "Transform (Modify) Bone - Bone: upperarm_r"(ComponentPose=@n44, Alpha=1, Translation="0,0,0", Rotation="0,0,0", Scale="1,1,1") {
    rotationMode=BMM_Replace
    rotationSpace=BCS_ParentBoneSpace
    translationMode=BMM_Replace
    translationSpace=BCS_ParentBoneSpace
    alphaBoolBlend {blendOption=Linear}
    boneToModify {boneName=upperarm_r}
   }
   n46 = "Transform (Modify) Bone - Bone: lowerarm_r"(ComponentPose=@n45, Alpha=1, Translation="0,0,0", Rotation="0,0,0", Scale="1,1,1") {
    rotationMode=BMM_Replace
    rotationSpace=BCS_ParentBoneSpace
    translationMode=BMM_Replace
    translationSpace=BCS_ParentBoneSpace
    alphaBoolBlend {blendOption=Linear}
    boneToModify {boneName=lowerarm_r}
   }
   n47 = "Transform (Modify) Bone - Bone: hand_r"(ComponentPose=@n46, Alpha=1, Translation="0,0,0", Rotation="0,0,0", Scale="1,1,1") {
    rotationMode=BMM_Replace
    rotationSpace=BCS_ParentBoneSpace
    translationMode=BMM_Replace
    translationSpace=BCS_ParentBoneSpace
    alphaBoolBlend {blendOption=Linear}
    boneToModify {boneName=hand_r}
   }
   n48 = Look At - Bone: dyn_upperarm_r(ComponentPose=@n47, Alpha=1) {
    alphaBoolBlend {blendOption=Linear}
    boneToModify {boneName=dyn_upperarm_r}
    lookAtLocation {x=0}
    lookAtTarget {boneReference {boneName=hand_r}}
    lookAt_Axis {axis=<-1 0>}
    lookUp_Axis {axis {x=-1}}
   }
   Save cached pose 'MainDynamicCache'(Pose=Component To Local(ComponentPose=@n48)) {cacheName=MainDynamicCache}
  ]}
 EventGraph {nodes [
   Event BlueprintUpdateAnimation()
   > Set Skydive hair fix(Skydive hair fix:bool=OR Boolean(A=Get bIsSkydiveDiveMode(), B=Get bIsSkydiveFloating())) -> (Output_Get:bool)
   n1 = GetOwnerSkeletalMeshComponent()
   n2 = BreakTransform(InTransform=GetSocketTransform(self=@n1, InSocketName=upperarm_r, TransformSpace=RTS_ParentBoneSpace))
   > Set Upperarm_r_Translates(Upperarm_r_Translates:Vector=@n2) -> (Output_Get:Vector)
   > Set Upperarm_r_Rotates(Upperarm_r_Rotates:Rotator=@n2.Rotation) -> (Output_Get:Rotator)
   n3 = BreakTransform(InTransform=GetSocketTransform(self=@n1, InSocketName=lowerarm_r, TransformSpace=RTS_ParentBoneSpace))
   > Set Lowerarm_r_Transtates(Lowerarm_r_Transtates:Vector=@n3) -> (Output_Get:Vector)
   > Set Lowerarm_r_Rotates(Lowerarm_r_Rotates:Rotator=@n3.Rotation) -> (Output_Get:Rotator)
   n4 = BreakTransform(InTransform=GetSocketTransform(self=@n1, InSocketName=hand_r, TransformSpace=RTS_ParentBoneSpace))
   > Set Hand_r_Translates(Hand_r_Translates:Vector=@n4) -> (Output_Get:Vector)
   > Set Hand_r_Rotates(Hand_r_Rotates:Rotator=@n4.Rotation) -> (Output_Get:Rotator)
   n5 = GetPartSkeletalMeshComponent()
   n6 = float + float(
    A=GetAlphaFromReferenceTransformDelta(SkeletalMeshComponent=@n5, BoneName=dyn_upperarm_r, BaseName=spine_05, InputRange=(X=80,Y=90), OutputRange=(X=0,Y=0.4), TwistAxis="1,0,0")
    B=GetAlphaFromReferenceTransformDelta(
     SkeletalMeshComponent=@n5
     BoneName=head
     BaseName=clavicle_r
     DeltaComponentType=EulerY
     InputRange=(X=15,Y=30)
     OutputRange=(X=0,Y=0.6)
     TwistAxis="1,0,0"
    )
   )
   n7 = NOT Boolean(A=Get bIsPlayingMeleeAnim())
   n8 = NOT Boolean(A=Get bIsSkyDiving())
   n9 = NOT Boolean(A=Get bIsPlayingEmote())
   > Set Targeting correction(Targeting correction:double=SelectFloat(A="Clamp (Float)"(Value=@n6, Min=0, Max=1), B=0, bPickA=AND Boolean!(A=@n8, B=@n9, C:bool=@n7) -> (ReturnValue))) -> (Output_Get:double)
   > Set Back Hair targeting up corr(Back Hair targeting up corr:float=GetAlphaFromReferenceTransformDelta(SkeletalMeshComponent=@n5, BoneName=head, BaseName=clavicle_r, InputRange=(X=-60,Y=-20), OutputRange=(X=1,Y=0), TwistAxis="1,0,0")) -> (Output_Get:float)
  ]}
]
variables [
 Back Hair targeting up corr {type=number}
 Hand_r_Rotates {type=Rotator}
 Hand_r_Translates {type=Vector}
 Lowerarm_r_Rotates {type=Rotator}
 Lowerarm_r_Transtates {type=Vector}
 Skydive hair fix {type=boolean}
 Targeting correction {type=number}
 Upperarm_r_Rotates {type=Rotator}
 Upperarm_r_Translates {type=Vector}
]
